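# Translucent unlit Niagara particle material for the ranked progression emote that fakes a decal on nearby scene surfaces.
# Fresnel from flattened scene world normals, with forward and mobile fallbacks, indexes a ranked color curve atlas, defaulting to Bronze.
# Three panning noise samples on scaled macro UVs form offset RGB channels, combined with curve and particle color for HDR emissive.
# Fresnel, particle alpha, sphere mask and inverted depth fade confine opacity near intersecting geometry.
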Material /FNGameplayCues/Effects/Fort_Effects/Effects/Characters/Generic/RankedProgression/M_Emote_RankedProgression_ParticleDecal
graph {
 properties {
  bUsedWithNanite=true
  bUsedWithNiagaraMeshParticles=true
  bUsedWithNiagaraSprites=true
  blendMode=BLEND_Translucent
  pixelDepthOffsetMode=PDOM_Legacy
  shadingModel=MSM_Unlit
  usedShadingModels=Unlit
 }
 nodes [
  n1 = MaterialFunctionCall(UVs=ParticleMacroUV(), Texture Scale=DynamicParameter() {paramNames [FresnelExponent, ScaleUVs, HDR Intensity, Param4]}.ScaleUVs) {materialFunction=&ScaleUVsByCenter}
  n2 = Add(A=@n1.UVs, B=NamedRerouteUsage())
  n3 = NamedRerouteUsage()
  n4 = NamedRerouteUsage()
  n5 = Add(A=@n1.UVs, B=Multiply(A=NamedRerouteUsage()) {constB=-1})
  n6 = NamedRerouteUsage()
  MaterialExpressionAppend3Vector_0 = MaterialXAppend3Vector!(
   A=TextureSampleParameter2D(UVs=Panner(Coordinate=@n2, Speed=@n3)) {parameterName=Texture, texture=/Game/Effects/Fort_Effects/Textures/Tile/T_GaussianNoise_01_M}.R
   B=TextureSampleParameter2D(UVs=Panner(Coordinate=@n1.UVs, Speed=@n4)) {parameterName=Texture, texture=&T_GaussianNoise_01_M}.R
   C=TextureSampleParameter2D(UVs=Panner(Coordinate=@n5, Speed=@n6)) {parameterName=Texture, texture=&T_GaussianNoise_01_M}.R
  )
  n7 = ParticleColor()
  n8 = SceneTexture() {sceneTextureId=PPI_WorldNormal}
  n9 = Fresnel(
   ExponentIn=DynamicParameter() {paramNames [FresnelExponent, ScaleUVs, HDR Intensity, Param4]}.FresnelExponent
   Normal=MaterialFunctionCall(Normal=ComponentMask(None=@n8.Color) {b=true, g=true, r=true}, Flatness=ScalarParameter() {defaultValue=0.2, parameterName=FlattenNormal}) {materialFunction=&FlattenNormal}.Result
  ) {baseReflectFraction=0, exponent=1}
  n10 = Add(A=ShadingPathSwitch(Default=@n9, Deferred=@n9, Forward=VectorParameter() {parameterName=Forward, defaultValue=<rgba 0 0 1 1>}.RGB, Mobile=Constant() {r=0.3})) {constB=0.25}
  n11 = CurveAtlasRowParameter(CurveTime=@n10) {
   atlas=&/CurveAtlas_RankedProgression
   curve=&/RankedProression_Bronze
   defaultValue=1
   parameterName=Curves
  }
  n12 = Add(A=Multiply(A=@MaterialExpressionAppend3Vector_0, B=@n7.RGB), B=Multiply(A=@n7.RGB, B=@n11))
  MaterialFunctionCall(Emissive=MaterialFunctionCall(
    LDR Input=Saturate(None=@n12)
    HDR Tint=@n12
    HDR Intensity Multiplier=DynamicParameter() {paramNames [FresnelExponent, ScaleUVs, HDR Intensity, Param4]}.HDR Intensity
   ) {materialFunction=&DeriveHDRfromLDR}.RemappedRange) {materialFunction=&MF_FX_EmissiveAdjustments}
  n13 = Multiply(A=@n10, B=ParticleColor().A)
  n14 = SphereMask(
   A=TextureCoordinate()
   B=Constant() {r=0.5}
   Radius=ScalarParameter() {defaultValue=0.5, parameterName=SphereMaskRadius}
   Hardness=ScalarParameter() {parameterName=SphereMaskHardness}
  ) {hardnessPercent=0}
  n15 = OneMinus(None=DepthFade(FadeDistance=ScalarParameter() {defaultValue=35, parameterName=VisibleDepthDistance}) {fadeDistanceDefault=80})
  n16 = ScalarParameter() {defaultValue=25, parameterName=FadeDistance}
  Multiply(
   A=Multiply(A=Multiply(A=@n13, B=@n14), B=DepthFade(Opacity=@n15, FadeDistance=@n16) {fadeDistanceDefault=25})
   B=ScalarParameter() {defaultValue=1, group=Blinking, parameterName=Opacity}
  )
  n17 = DynamicParameter() {parameterIndex=1, defaultValue=<rgba 25 1 0 0>, paramNames [Panner X Speed, Panner Y Speed, X Chromatic Offset, Y Chromatic Offset]}
  NamedRerouteDeclaration(None=AppendVector(A=@n17.Panner X Speed, B=@n17.Panner X Speed)) {name=Panner Speed, nodeColor=<rgba 0 1 0.6016 1>}
  NamedRerouteDeclaration(None=AppendVector(A=@n17.X Chromatic Offset, B=@n17.X Chromatic Offset)) {name=ChromaticOffset, nodeColor=<rgba 1 0.8203 0 1>}
 ]
}
