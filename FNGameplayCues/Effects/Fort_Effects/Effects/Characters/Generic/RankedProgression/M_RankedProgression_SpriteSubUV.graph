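# Unlit translucent Niagara sprite and mesh-particle material for flame effects in ranked progression, colored through a rank curve atlas that defaults to Bronze.
# A packed SubUV flipbook with a selectable channel is modulated by panning tiled flame detail, with per-particle offsets and scaling from dynamic parameters.
# The result indexes the color curve, tinted by particle color and HDR-boosted, and also shapes depth-faded opacity.
# Low quality uses simpler brightness and opacity paths.
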Material /FNGameplayCues/Effects/Fort_Effects/Effects/Characters/Generic/RankedProgression/M_RankedProgression_SpriteSubUV
graph {
 properties {
  bCastRayTracedShadows=false
  bUsedWithNiagaraMeshParticles=true
  bUsedWithNiagaraSprites=true
  blendMode=BLEND_Translucent
  pixelDepthOffsetMode=PDOM_Legacy
  shadingModel=MSM_Unlit
  usedShadingModels=Unlit
 }
 nodes [
  n1 = DynamicParameter() {parameterIndex=2, paramNames [Ignition Normal.X, Ignition Normal.Y, Ignition Normal.Z, Size]}
  n2 = DynamicParameter() {parameterIndex=1, paramNames [Sprite_RandomOffset, Sprite_TimeOffset, Detail_Intensity, HDR Intensity]}
  n3 = Add(
   A=Multiply(A=TextureCoordinate(), B=@n1.Size)
   B=MaterialFunctionCall(X=@n2.Sprite_RandomOffset, Y=@n2.Sprite_TimeOffset) {materialFunction=&MakeFloat2}.Result
  )
  n4 = ScalarParameter() {defaultValue=1, parameterName=Detail TexScale, sliderMax=1}
  n5 = Time()
  n6 = Multiply(A=@n2.Sprite_RandomOffset) {constB=25}
  n7 = Constant()
  n8 = Multiply(A=@n1.Size, B=ScalarParameter() {defaultValue=2, parameterName=Detail PanSpeed, sliderMax=1})
  n9 = TextureSample(UVs=Panner(Coordinate=Multiply(A=@n3, B=@n4) {constB=0.4}, Time=Add(A=@n5, B=@n6), Speed=AppendVector(A=@n7, B=@n8)) {speedY=0.05}) {samplerType=SAMPLERTYPE_LinearGrayscale, texture=/Game/Weapons/FORT_Melee/Pickaxe_Cube_Ninja_Male/FX/T_FluidFlameTile}
  n10 = ComponentMask(None=Multiply(
    A=DynamicParameter() {parameterIndex=1, paramNames [Sprite_RandomOffset, Sprite_TimeOffset, Detail_Intensity, HDR Intensity]}.Detail_Intensity
    B=Power(Base=@n9.R) {constExponent=0.5}
   ) {constA=0.5}) {r=true}
  CurveAtlasRowParameter(CurveTime=@n10) {
   atlas=/Game/Curie/Effects/temp/Curves/CA_Curie_Fire_Visuals
   curve=/Game/Curie/Effects/temp/Curves/Curve_Fiery_Orange
   defaultValue=1
   parameterName=CurveAtlas_fire
  }
  n11 = ParticleColor()
  n12 = NamedRerouteUsage()
  n13 = Multiply(
   A=@n12
   B=LinearInterpolate(
    A=ScalarParameter() {defaultValue=0.5, parameterName=Detail Min, sliderMax=3, sliderMin=-1}
    B=ScalarParameter() {defaultValue=1.3, parameterName=Detail Max, sliderMax=3, sliderMin=-1}
    Alpha=NamedRerouteUsage()
   ) {constA=0.8, constB=2}
  )
  n14 = CurveAtlasRowParameter(CurveTime=@n13) {
   atlas=&/CurveAtlas_RankedProgression
   curve=&/RankedProression_Bronze
   defaultValue=1
   parameterName=Curves
  }
  n15 = Multiply(A=@n11.RGB, B=@n14)
  n16 = Saturate(None=@n15)
  n17 = DynamicParameter() {parameterIndex=1, paramNames [Sprite_RandomOffset, Sprite_TimeOffset, Detail_Intensity, HDR Intensity]}
  n18 = ScalarParameter() {defaultValue=2, group=Output Color, parameterName=Mobile Brightness, sliderMax=50, sortPriority=9}
  MaterialFunctionCall(Emissive=QualitySwitch(
    Default=MaterialFunctionCall(LDR Input=@n16, HDR Tint=@n15, HDR Intensity Multiplier=@n17.HDR Intensity) {materialFunction=&DeriveHDRfromLDR}.RemappedRange
    Low=Multiply(A=@n15, B=@n18)
   ) {desc=quality RGB}) {materialFunction=&MF_FX_EmissiveAdjustments}
  n19 = TextureObjectParameter() {group=Flame Channels, parameterName=Texture, sortPriority=1, texture=/Game/Effects/Fort_Effects/Textures/Fire/2DGasTorch_Flipbk_packed}
  n20 = MaterialFunctionCall(Float2=TextureCoordinate()) {materialFunction=&BreakOutFloat2Components}
  n21 = DynamicParameter() {paramNames [Temperature Channel Scale, Packed Channel, SubUV_Images, SubUV_Frame]}
  n22 = MaterialFunctionCall(Float3=MaterialFunctionCall!(
    Texture=@n19
    UVs=MaterialFunctionCall(X=@n20.R, Y=@n20.R) {materialFunction=&MakeFloat2}.Result
    SubImages=AppendVector(A=@n21.SubUV_Images, B=@n21.SubUV_Images)
    Frame=@n21.SubUV_Frame
   ) {materialFunction=/Engine/Functions/Engine_MaterialFunctions03/Texturing/SubUV_Function_MipDerivative}.RGB) {materialFunction=&BreakOutFloat3Components}
  n23 = DynamicParameter() {desc=G - packed channel switch/weight, paramNames [Temperature Channel Scale, Packed Channel, SubUV_Images, SubUV_Frame]}
  n24 = Multiply(
   A=Multiply(
    A=@n23.Temperature Channel Scale
    B=If(A=@n23.Packed Channel, "A > B"=@n22.B, "A == B"=@n22.G, "A < B"=@n22.R) {constB=1, desc=pick flipbook channels, equalsThreshold=0.1}
   )
   B=ScalarParameter() {defaultValue=1.8, group=Flame Channels, parameterName=Temperature Channel Scale, sliderMax=10, sortPriority=3}
  )
  n25 = ScalarParameter() {defaultValue=1.25, group=Flame Channels, parameterName=Temperature Exponent, sliderMax=3, sortPriority=5}
  NamedRerouteDeclaration(None=Multiply(
    A=Power(Base=@n24, Exp=@n25)
    B=ScalarParameter() {defaultValue=2, group=Flame Channels, parameterName=Temperature Scale, sliderMax=25, sortPriority=5}
   )) {name=Flipbook Flame Texture, nodeColor=<rgba 1 0 0.0937 1>}
  NamedRerouteDeclaration(None=@n10) {name=Tiling Flame Texture, nodeColor=<rgba 1 0 0.3516 1>}
  n26 = Power(
   Base=QualitySwitch(Default=@n13, Low=@n12, Epic=@n13) {desc=quality Opactiy}
   Exp=ScalarParameter() {defaultValue=1.5, group=Output Color, parameterName=Opacity Exponent, sliderMax=3, sortPriority=8}
  )
  n27 = ScalarParameter() {defaultValue=1.5, group=Output Color, parameterName=Opacity Scale, sliderMax=3, sortPriority=8}
  n28 = Saturate(None=Multiply(A=@n11.A, B=Multiply(A=@n26, B=@n27) {constB=3}))
  n29 = DepthFade(Opacity=@n28, FadeDistance=ScalarParameter() {defaultValue=55, parameterName=FadeDistance, sliderMax=300})
  Reroute(None=QualitySwitch(Default=@n29, Low=@n28, Epic=@n29) {desc=quality Opactiy})
 ]
}
