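# Translucent unlit Niagara ribbon material for ranked progression effects, drawing animated flames tinted by a curve atlas row, Bronze by default.
# AnimTime drives three 8x8 flipbook reads of a packed torch texture at slightly offset rates on swapped UVs, merged by Max.
# The shaped result indexes the curve for particle-tinted HDR emissive, and also forms opacity with particle alpha.
# A quality switch skips depth fade on Low.
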Material /FNGameplayCues/Effects/Fort_Effects/Effects/Characters/Generic/RankedProgression/M_RankedProgression_SubUV_Ribbon
graph {
 properties {bUsedWithNiagaraRibbons=true, blendMode=BLEND_Translucent, pixelDepthOffsetMode=PDOM_Legacy, shadingModel=MSM_Unlit, usedShadingModels=Unlit}
 nodes [
  n1 = ParticleColor()
  n2 = DynamicParameter() {desc=G - packed channel switch/weight, defaultValue=<rgba 0.5 1 0 0>, paramNames [AnimTime, HDR Intensity, Param3, Param4]}
  n3 = Multiply(A=@n2.AnimTime) {constB=1.1}
  n4 = Constant() {r=8}
  n5 = MaterialFunctionCall(XY=TextureCoordinate()) {materialFunction=&Swizzle}
  n6 = Multiply(A=@n2.AnimTime) {constB=0.9}
  n7 = Max(
   A=TextureSample(UVs=MaterialFunctionCall("Animation  Phase (0-1)"=@n3, Number of Rows=@n4, Number of Columns=@n4, UVs=@n5.YX) {materialFunction=&FlipBook}.UVs) {texture=/Game/Effects/Fort_Effects/Textures/Fire/2DGasTorch_Flipbk_packed}.R
   B=TextureSample(UVs=MaterialFunctionCall("Animation  Phase (0-1)"=@n6, Number of Rows=@n4, Number of Columns=@n4, UVs=@n5.YX) {materialFunction=&FlipBook}.UVs) {texture=&2DGasTorch_Flipbk_packed}.G
  )
  n8 = TextureSample(UVs=MaterialFunctionCall("Animation  Phase (0-1)"=@n2.AnimTime, Number of Rows=@n4, Number of Columns=@n4, UVs=@n5.YX) {materialFunction=&FlipBook}.UVs) {texture=&2DGasTorch_Flipbk_packed}
  n9 = Multiply(
   A=Power(
    Base=Max(A=@n7, B=@n8.B)
    Exp=ScalarParameter() {defaultValue=1.25, group=Flame Channels, parameterName=Temperature Exponent, sliderMax=3, sortPriority=5}
   )
   B=ScalarParameter() {defaultValue=5, group=Flame Channels, parameterName=Temperature Scale, sliderMax=25, sortPriority=5}
  )
  n10 = CurveAtlasRowParameter(CurveTime=@n9) {
   atlas=&/CurveAtlas_RankedProgression
   curve=&/RankedProression_Bronze
   defaultValue=1
   parameterName=Curve
  }
  n11 = Multiply(A=@n1.RGB, B=@n10)
  n12 = DynamicParameter() {desc=G - packed channel switch/weight, defaultValue=<rgba 0.5 1 0 0>, paramNames [AnimTime, HDR Intensity, Param3, Param4]}
  MaterialFunctionCall(Emissive=MaterialFunctionCall(LDR Input=Saturate(None=@n11), HDR Tint=@n11, HDR Intensity Multiplier=@n12.HDR Intensity) {materialFunction=&DeriveHDRfromLDR}.RemappedRange) {materialFunction=&MF_FX_EmissiveAdjustments}
  n13 = Power(Base=@n9, Exp=ScalarParameter() {defaultValue=1.5, group=Output Color, parameterName=Opacity Exponent, sliderMax=3, sortPriority=8})
  n14 = ScalarParameter() {defaultValue=3, group=Output Color, parameterName=Opacity Scale, sliderMax=3, sortPriority=8}
  n15 = Saturate(None=Multiply(A=@n1.A, B=Multiply(A=@n13, B=@n14) {constB=3}))
  n16 = DepthFade(Opacity=@n15, FadeDistance=ScalarParameter() {defaultValue=15, parameterName=FadeDistance, sliderMax=300})
  Reroute(None=QualitySwitch(Default=@n16, Low=@n15, Epic=@n16) {desc=quality Opactiy})
 ]
}
