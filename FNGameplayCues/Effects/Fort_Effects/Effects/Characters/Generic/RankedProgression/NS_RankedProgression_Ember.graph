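# Ranked progression "ember" Niagara effect: a glowing, source-following sprite mote with sparkle dust, colored by rank tier from Unranked to Unreal.
# Tinkerbell forms the core from two counter-rotating spawn groups, and TinkDust trails flickering particles from it via a particle attribute reader.
# Corona_Burst adds a brief flash and TinkDust_Burst throws sparks upward and outward under gravity and noise.
# User.Ranked_Curve selects each emitter's tier color curve through SelectColorCurve.
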NiagaraSystem /FNGameplayCues/Effects/Fort_Effects/Effects/Characters/Generic/RankedProgression/NS_RankedProgression_Ember
dependencies {
 usedDataInterfaces [&NiagaraDataInterfaceColorCurve, &NiagaraDataInterfaceCurve, &NiagaraDataInterfaceParticleRead]
 usedDynamicInputs []
 usedModules [&SystemState]
}
emitters {
 Corona_Burst {
  emitterUpdateScript [
   &EmitterState(Life Cycle Mode=::NewEnumerator1, Loop Behavior=::NewEnumerator1, Loop Duration=0.1, UseLoopDelay=false)
   &SpawnBurst_Instantaneous(Use Spawn Probability=false, UseLoopCountLimit=false)
   &SpawnBurst_Instantaneous(Spawn Group=1, Use Spawn Probability=false, UseLoopCountLimit=false) {module=SpawnBurst_Instantaneous001}
  ]
  particleSpawnScript [&InitializeParticle(
    Color=&MaskLinearColorBySpawnGroup(
     Masked Linear Color=&self:SelectColorCurve(
      CurveSelect:NiagaraFloat=$User.Ranked_Curve
      ColorSelect:NiagaraFloat=0
      UnrankedColorCurve:NiagaraDataInterfaceColorCurve=DataInterface() {propertyValues {AlphaCurve {keys []}, BlueCurve {keys []}, GreenCurve {keys []}, RedCurve {keys []}}}
      BronzeColorCurve:NiagaraDataInterfaceColorCurve=DataInterface() {propertyValues {
        AlphaCurve {keys [<at 0 1>,<at 1 1>]}
        BlueCurve {keys [<at 0 0.0185>,<at 1 0.1946>]}
        GreenCurve {keys [<at 0 0.0714>,<at 1 0.3557>]}
        RedCurve {keys [<at 0 0.2831>,<at 1 1>]}
       }}
      SilverColorCurve:NiagaraDataInterfaceColorCurve=DataInterface() {propertyValues {
        AlphaCurve {keys [<at 0 1>,<at 1 1>]}
        BlueCurve {keys [<at 0 0.0409>,<at 1 0.9734>]}
        GreenCurve {keys [<at 0 0.0273>,<at 1 0.9301>]}
        RedCurve {keys [<at 0 0.0194>,<at 1 0.8796>]}
       }}
      GoldColorCurve:NiagaraDataInterfaceColorCurve=DataInterface() {propertyValues {
        AlphaCurve {keys [<at 0 1>,<at 1 1>]}
        BlueCurve {keys [<at 0 0>,<at 1 0.2961>]}
        GreenCurve {keys [<at 0 0.1694>,<at 1 0.7654>]}
        RedCurve {keys [<at 0 0.2542>,<at 1 1>]}
       }}
      PlatColorCurve:NiagaraDataInterfaceColorCurve=DataInterface() {propertyValues {
        AlphaCurve {keys [<at 0 1>,<at 1 1>]}
        BlueCurve {keys [<at 0 0.0221>,<at 1 0.404>]}
        GreenCurve {keys [<at 0 0.05>,<at 1 0.9216>]}
        RedCurve {keys [<at 0 0.0021>,<at 1 0.0343>]}
       }}
      DiamondColorCurve:NiagaraDataInterfaceColorCurve=DataInterface() {propertyValues {
        AlphaCurve {keys [<at 0 1>,<at 1 1>]}
        BlueCurve {keys [<at 0 0.4>,<at 1 1>]}
        GreenCurve {keys [<at 0 0.2356>,<at 1 0.5833>]}
        RedCurve {keys [<at 0 0.0054>,<at 1 0>]}
       }}
      EliteColorCurve:NiagaraDataInterfaceColorCurve=DataInterface() {propertyValues {
        AlphaCurve {keys [<at 0 1>,<at 1 1>]}
        BlueCurve {keys [<at 0 0.0005 RCIM_Cubic>, {arriveTangent=0.3468, interpMode=RCIM_Cubic, leaveTangent=0.3468, tangentMode=RCTM_User, time=1, value=0.1}]}
        GreenCurve {keys [<at 0 0.0005 RCIM_Cubic>, {arriveTangent=0.3468, interpMode=RCIM_Cubic, leaveTangent=0.3468, tangentMode=RCTM_User, time=1, value=0.1}]}
        RedCurve {keys [<at 0 0.0005 RCIM_Cubic>, {arriveTangent=0.3468, interpMode=RCIM_Cubic, leaveTangent=0.3468, tangentMode=RCTM_User, time=1, value=0.1}]}
       }}
      ChampionColorCurve:NiagaraDataInterfaceColorCurve=DataInterface() {propertyValues {
        AlphaCurve {keys [<at 0 1>,<at 1 1>]}
        BlueCurve {keys [<at 0 0>,<at 0.5 0>,<at 1 0.0012>]}
        GreenCurve {keys [<at 0 0 RCIM_Cubic>, {arriveTangent=0.6654, interpMode=RCIM_Cubic, leaveTangent=0.6654, time=0.5, value=0.0908}, <at 1 0.6654 RCIM_Cubic>]}
        RedCurve {keys [
          <at 0 0.15 RCIM_Cubic>
          {arriveTangentWeight=0.2187, interpMode=RCIM_Cubic, leaveTangentWeight=0.2187, tangentMode=RCTM_User, time=0.5, value=1}
          <at 1 1 RCIM_Cubic>
         ]}
       }}
      UnrealColorCurve:NiagaraDataInterfaceColorCurve=DataInterface() {propertyValues {
        AlphaCurve {keys [<at 0 1>,<at 1 1>]}
        BlueCurve {keys [<at 0 0.2307 RCIM_Cubic>, {interpMode=RCIM_Cubic, leaveTangent=0.5013, time=1, value=0.2664}]}
        GreenCurve {keys [<at 0 0.0252 RCIM_Cubic>, {interpMode=RCIM_Cubic, leaveTangent=0.6649, time=1, value=0.0782}]}
        RedCurve {keys [
          <at 0 0.0222 RCIM_Cubic>
          {arriveTangentWeight=0.2187, interpMode=RCIM_Cubic, leaveTangent=0.4814, leaveTangentWeight=0.2187, time=1, value=0.3564}
         ]}
       }}
     )
     Passthrough Linear Color=&self:SelectColorCurve(
      CurveSelect:NiagaraFloat=$User.Ranked_Curve
      ColorSelect:NiagaraFloat=1
      UnrankedColorCurve:NiagaraDataInterfaceColorCurve=DataInterface() {propertyValues {
        AlphaCurve {keys [<at 0 1>,<at 1 1>]}
        BlueCurve {keys [
          {interpMode=RCIM_Cubic, leaveTangentWeight=0.4171, tangentMode=RCTM_User, time=0, value=0.2831}
          {arriveTangent=1.9677, interpMode=RCIM_Cubic, leaveTangent=1.9677, tangentMode=RCTM_User, time=1, value=1}
         ]}
        GreenCurve {keys [
          {interpMode=RCIM_Cubic, leaveTangentWeight=0.4171, tangentMode=RCTM_User, time=0, value=0.2831}
          {arriveTangent=1.9677, interpMode=RCIM_Cubic, leaveTangent=1.9677, tangentMode=RCTM_User, time=1, value=1}
         ]}
        RedCurve {keys [
          {interpMode=RCIM_Cubic, leaveTangentWeight=0.4171, tangentMode=RCTM_User, time=0, value=0.2831}
          {arriveTangent=1.9677, interpMode=RCIM_Cubic, leaveTangent=1.9677, tangentMode=RCTM_User, time=1, value=1}
         ]}
       }}
      BronzeColorCurve:NiagaraDataInterfaceColorCurve=DataInterface() {propertyValues {
        AlphaCurve {keys [<at 0 1>,<at 1 1>]}
        BlueCurve {keys [<at 0 0.0185>,<at 1 0.1946>]}
        GreenCurve {keys [<at 0 0.0714>,<at 1 0.3557>]}
        RedCurve {keys [<at 0 0.2831>,<at 1 1>]}
       }}
      SilverColorCurve:NiagaraDataInterfaceColorCurve=DataInterface() {propertyValues {
        AlphaCurve {keys [<at 0 1>,<at 1 1>]}
        BlueCurve {keys [<at 0 0.0409>,<at 1 0.9734>]}
        GreenCurve {keys [<at 0 0.0273>,<at 1 0.9301>]}
        RedCurve {keys [<at 0 0.0194>,<at 1 0.8796>]}
       }}
      GoldColorCurve:NiagaraDataInterfaceColorCurve=DataInterface() {propertyValues {
        AlphaCurve {keys [<at 0 1>,<at 1 1>]}
        BlueCurve {keys [<at 0 0>,<at 1 0.2961>]}
        GreenCurve {keys [<at 0 0.1694>,<at 1 0.7654>]}
        RedCurve {keys [<at 0 0.2542>,<at 1 1>]}
       }}
      PlatColorCurve:NiagaraDataInterfaceColorCurve=DataInterface() {propertyValues {
        AlphaCurve {keys [<at 0 1>,<at 1 1>]}
        BlueCurve {keys [<at 0 0.0221>,<at 1 0.404>]}
        GreenCurve {keys [<at 0 0.05>,<at 1 0.9216>]}
        RedCurve {keys [<at 0 0.0021>,<at 1 0.0343>]}
       }}
      DiamondColorCurve:NiagaraDataInterfaceColorCurve=DataInterface() {propertyValues {
        AlphaCurve {keys [<at 0 1>,<at 1 1>]}
        BlueCurve {keys [<at 0 0.4>,<at 1 1>]}
        GreenCurve {keys [<at 0 0.2356>,<at 1 0.5833>]}
        RedCurve {keys [<at 0 0.0054>,<at 1 0>]}
       }}
      EliteColorCurve:NiagaraDataInterfaceColorCurve=DataInterface() {propertyValues {
        AlphaCurve {keys [<at 0 1>,<at 1 1>]}
        BlueCurve {keys [<at 0 0.0005 RCIM_Cubic>, {arriveTangent=0.3468, interpMode=RCIM_Cubic, leaveTangent=0.3468, tangentMode=RCTM_User, time=1, value=0.1}]}
        GreenCurve {keys [<at 0 0.0005 RCIM_Cubic>, {arriveTangent=0.3468, interpMode=RCIM_Cubic, leaveTangent=0.3468, tangentMode=RCTM_User, time=1, value=0.1}]}
        RedCurve {keys [<at 0 0.0005 RCIM_Cubic>, {arriveTangent=0.3468, interpMode=RCIM_Cubic, leaveTangent=0.3468, tangentMode=RCTM_User, time=1, value=0.1}]}
       }}
      ChampionColorCurve:NiagaraDataInterfaceColorCurve=DataInterface() {propertyValues {
        AlphaCurve {keys [<at 0 1>,<at 1 1>]}
        BlueCurve {keys [<at 0 0>,<at 0.5 0>,<at 1 0.0012>]}
        GreenCurve {keys [<at 0 0 RCIM_Cubic>, {arriveTangent=0.6654, interpMode=RCIM_Cubic, leaveTangent=0.6654, time=0.5, value=0.0908}, <at 1 0.6654 RCIM_Cubic>]}
        RedCurve {keys [
          <at 0 0.15 RCIM_Cubic>
          {arriveTangentWeight=0.2187, interpMode=RCIM_Cubic, leaveTangentWeight=0.2187, tangentMode=RCTM_User, time=0.5, value=1}
          <at 1 1 RCIM_Cubic>
         ]}
       }}
      UnrealColorCurve:NiagaraDataInterfaceColorCurve=DataInterface() {propertyValues {
        AlphaCurve {keys [<at 0 1>,<at 1 1>]}
        BlueCurve {keys [<at 0 0.2307 RCIM_Cubic>, {interpMode=RCIM_Cubic, leaveTangent=0.5013, time=1, value=0.2664}]}
        GreenCurve {keys [<at 0 0.0252 RCIM_Cubic>, {interpMode=RCIM_Cubic, leaveTangent=0.6649, time=1, value=0.0782}]}
        RedCurve {keys [
          <at 0 0.0222 RCIM_Cubic>
          {arriveTangentWeight=0.2187, interpMode=RCIM_Cubic, leaveTangent=0.4814, leaveTangentWeight=0.2187, time=1, value=0.3564}
         ]}
       }}
     )
    )
    Lifetime=0.2
    Sprite Rotation Mode=::NewEnumerator2
    Sprite Size Mode=::NewEnumerator3
    Uniform Sprite Size=60
    UsePositionOffset=false
    Write Material Random=true
   )]
  particleUpdateScript [
   &ParticleState()
   &ScaleSpriteSize(Uniform Curve Sprite Scale=DataInterface() {propertyValues {Curve {keys [
        {interpMode=RCIM_Cubic, leaveTangentWeight=0.9682, tangentMode=RCTM_User, time=0, value=0}
        {arriveTangent=0.3352, arriveTangentWeight=0.3333, interpMode=RCIM_Cubic, leaveTangent=0.3352, tangentMode=RCTM_User, time=1, value=1}
       ]}}})
   &DynamicMaterialParameters(
    Index 0 Param 1=&FloatFromCurve(FloatCurve=DataInterface() {propertyValues {Curve {keys [<at 0.8 1 RCIM_Cubic>, {arriveTangent=-5.1423, interpMode=RCIM_Cubic, leaveTangent=-5.1423, tangentMode=RCTM_User, time=1, value=0}]}}})
    Index 0 Param 2=1
    Index 0 Param 3=1
    Index 0 Param 4=1
    Module.Param0WriteEnabled=true
    Module.Param1WriteEnabled=true
    Module.Param2WriteEnabled=true
    Module.Param3WriteEnabled=true
   )
   &ScaleColor(
    Scale Alpha=&FloatFromCurve(FloatCurve=DataInterface() {propertyValues {Curve {keys [{arriveTangent=-5.8164, interpMode=RCIM_Cubic, leaveTangent=-5.8164, tangentMode=RCTM_User, time=0.75, value=1}, <at 1 0 RCIM_Cubic>]}}})
    Scale RGB=&VectorFromFloat(Value=&FloatFromCurve(FloatCurve=DataInterface() {propertyValues {Curve {keys [{arriveTangent=-62.717, interpMode=RCIM_Cubic, leaveTangent=-62.717, tangentMode=RCTM_User, time=0, value=10}, <at 0.3 1 RCIM_Cubic>]}}}))
    ScaleA=true
    ScaleRGB=true
   )
  ]
  rendererClasses [&NiagaraSpriteRendererProperties]
  renderers [{rendererClass=&NiagaraSpriteRendererProperties}]
 }
 TinkDust {
  emitterUpdateScript [
   &EmitterState(Life Cycle Mode=::NewEnumerator1, Loop Behavior=::NewEnumerator1, Loop Duration=1.5, UseLoopDelay=false)
   &SpawnParticlesFromOtherEmitter(
    Attribute Reader=DataInterface() {propertyValues {EmitterBinding {bindingMode=Other, emitterName=Tinkerbell}}}
    Spawn Rate=100
    Spawn Rate Per Particle Cap=500
    Use Maximum Age=false
    Use Minimum Age=false
   )
  ]
  particleSpawnScript [
   &InitializeParticle(
    Color=&self:SelectColorCurve(
     CurveSelect:NiagaraFloat=$User.Ranked_Curve
     ColorSelect:NiagaraFloat=&RandomRangeFloat(Minimum=0, Maximum=1)
     UnrankedColorCurve:NiagaraDataInterfaceColorCurve=DataInterface() {propertyValues {
       AlphaCurve {keys [<at 0 1>,<at 1 1>]}
       BlueCurve {keys [
         {interpMode=RCIM_Cubic, leaveTangentWeight=0.4171, tangentMode=RCTM_User, time=0, value=0.2831}
         {arriveTangent=1.9677, interpMode=RCIM_Cubic, leaveTangent=1.9677, tangentMode=RCTM_User, time=1, value=1}
        ]}
       GreenCurve {keys [
         {interpMode=RCIM_Cubic, leaveTangentWeight=0.4171, tangentMode=RCTM_User, time=0, value=0.2831}
         {arriveTangent=1.9677, interpMode=RCIM_Cubic, leaveTangent=1.9677, tangentMode=RCTM_User, time=1, value=1}
        ]}
       RedCurve {keys [
         {interpMode=RCIM_Cubic, leaveTangentWeight=0.4171, tangentMode=RCTM_User, time=0, value=0.2831}
         {arriveTangent=1.9677, interpMode=RCIM_Cubic, leaveTangent=1.9677, tangentMode=RCTM_User, time=1, value=1}
        ]}
      }}
     BronzeColorCurve:NiagaraDataInterfaceColorCurve=DataInterface() {propertyValues {
       AlphaCurve {keys [<at 0 1>,<at 1 1>]}
       BlueCurve {keys [<at 0 0.0185>,<at 1 0.1946>]}
       GreenCurve {keys [<at 0 0.0714>,<at 1 0.3557>]}
       RedCurve {keys [<at 0 0.2831>,<at 1 1>]}
      }}
     SilverColorCurve:NiagaraDataInterfaceColorCurve=DataInterface() {propertyValues {
       AlphaCurve {keys [<at 0 1>,<at 1 1>]}
       BlueCurve {keys [<at 0 0.0409>,<at 1 0.9734>]}
       GreenCurve {keys [<at 0 0.0273>,<at 1 0.9301>]}
       RedCurve {keys [<at 0 0.0194>,<at 1 0.8796>]}
      }}
     GoldColorCurve:NiagaraDataInterfaceColorCurve=DataInterface() {propertyValues {
       AlphaCurve {keys [<at 0 1>,<at 1 1>]}
       BlueCurve {keys [<at 0 0>,<at 1 0.2961>]}
       GreenCurve {keys [<at 0 0.1694>,<at 1 0.7654>]}
       RedCurve {keys [<at 0 0.2542>,<at 1 1>]}
      }}
     PlatColorCurve:NiagaraDataInterfaceColorCurve=DataInterface() {propertyValues {
       AlphaCurve {keys [<at 0 1>,<at 1 1>]}
       BlueCurve {keys [<at 0 0.0221>,<at 1 0.404>]}
       GreenCurve {keys [<at 0 0.05>,<at 1 0.9216>]}
       RedCurve {keys [<at 0 0.0021>,<at 1 0.0343>]}
      }}
     DiamondColorCurve:NiagaraDataInterfaceColorCurve=DataInterface() {propertyValues {
       AlphaCurve {keys [<at 0 1>,<at 1 1>]}
       BlueCurve {keys [<at 0 0.4>,<at 1 1>]}
       GreenCurve {keys [<at 0 0.2356>,<at 1 0.5833>]}
       RedCurve {keys [<at 0 0.0054>,<at 1 0>]}
      }}
     EliteColorCurve:NiagaraDataInterfaceColorCurve=DataInterface() {propertyValues {
       AlphaCurve {keys [<at 0 1>,<at 1 1>]}
       BlueCurve {keys [<at 0 0.0005 RCIM_Cubic>, {arriveTangent=0.3468, interpMode=RCIM_Cubic, leaveTangent=0.3468, tangentMode=RCTM_User, time=1, value=0.1}]}
       GreenCurve {keys [<at 0 0.0005 RCIM_Cubic>, {arriveTangent=0.3468, interpMode=RCIM_Cubic, leaveTangent=0.3468, tangentMode=RCTM_User, time=1, value=0.1}]}
       RedCurve {keys [<at 0 0.0005 RCIM_Cubic>, {arriveTangent=0.3468, interpMode=RCIM_Cubic, leaveTangent=0.3468, tangentMode=RCTM_User, time=1, value=0.1}]}
      }}
     ChampionColorCurve:NiagaraDataInterfaceColorCurve=DataInterface() {propertyValues {
       AlphaCurve {keys [<at 0 1>,<at 1 1>]}
       BlueCurve {keys [<at 0 0>,<at 0.5 0>,<at 1 0.0012>]}
       GreenCurve {keys [<at 0 0 RCIM_Cubic>, {arriveTangent=0.6654, interpMode=RCIM_Cubic, leaveTangent=0.6654, time=0.5, value=0.0908}, <at 1 0.6654 RCIM_Cubic>]}
       RedCurve {keys [
         <at 0 0.15 RCIM_Cubic>
         {arriveTangentWeight=0.2187, interpMode=RCIM_Cubic, leaveTangentWeight=0.2187, tangentMode=RCTM_User, time=0.5, value=1}
         <at 1 1 RCIM_Cubic>
        ]}
      }}
     UnrealColorCurve:NiagaraDataInterfaceColorCurve=DataInterface() {propertyValues {
       AlphaCurve {keys [<at 0 1>,<at 1 1>]}
       BlueCurve {keys [<at 0 0.2307 RCIM_Cubic>, {interpMode=RCIM_Cubic, leaveTangent=0.5013, time=1, value=0.2664}]}
       GreenCurve {keys [<at 0 0.0252 RCIM_Cubic>, {interpMode=RCIM_Cubic, leaveTangent=0.6649, time=1, value=0.0782}]}
       RedCurve {keys [
         <at 0 0.0222 RCIM_Cubic>
         {arriveTangentWeight=0.2187, interpMode=RCIM_Cubic, leaveTangent=0.4814, leaveTangentWeight=0.2187, time=1, value=0.3564}
        ]}
      }}
    )
    Lifetime Max=0.75
    Lifetime Min=0.25
    Lifetime Mode=::NewEnumerator1
    Sprite Size Mode=::NewEnumerator1
    Uniform Sprite Size Max=14
    Uniform Sprite Size Min=2
    UsePositionOffset=false
    Write Material Random=true
   )
   &SampleParticlesFromOtherEmitter()
   &ShapeLocation(
    Hemisphere Distribution=<1 0.5>
    MaskSpawnGroup=false
    Override Seed=false
    Rotation Axis=<0 0 -1>
    Rotation Mode=::NewEnumerator1
    Sphere Radius=2
    UseOwnerScale=true
   )
  ]
  particleUpdateScript [
   &ParticleState()
   &ScaleSpriteSize(Uniform Curve Sprite Scale=DataInterface() {propertyValues {Curve {keys [
        <at 0 1 RCIM_Cubic>
        {arriveTangentWeight=0.0462, interpMode=RCIM_Cubic, leaveTangentWeight=0.0449, tangentMode=RCTM_User, time=0.1354, value=0.2886}
        {arriveTangentWeight=0.0512, interpMode=RCIM_Cubic, leaveTangentWeight=0.0563, tangentMode=RCTM_User, time=0.2671, value=0.9417}
        {arriveTangentWeight=0.0484, interpMode=RCIM_Cubic, leaveTangentWeight=0.0536, tangentMode=RCTM_User, time=0.412, value=0.1231}
        {arriveTangentWeight=0.0535, interpMode=RCIM_Cubic, leaveTangentWeight=0.0475, tangentMode=RCTM_User, time=0.5725, value=0.9591}
        {arriveTangentWeight=0.0479, interpMode=RCIM_Cubic, leaveTangentWeight=0.0524, tangentMode=RCTM_User, time=0.7149, value=0.1294}
        {arriveTangentWeight=0.0536, interpMode=RCIM_Cubic, leaveTangentWeight=0.0443, tangentMode=RCTM_User, time=0.8709, value=0.9199}
        {arriveTangentWeight=0.043, interpMode=RCIM_Cubic, tangentMode=RCTM_User, time=1, value=0.2}
       ]}}})
   &AccelerationForce(Acceleration=<0 0 -500>)
   &VectorNoiseForce(Force Amount=500, Module.Use Random Noise Time Interval=false)
   &SolveForcesAndVelocity(Clamp Velocity=false, Limit Acceleration=false)
   &ScaleColor(
    Scale RGB=&VectorFromFloat(Value=&Multiply_Float(
      A=&RandomRangeFloat(Minimum=2, Maximum=8, Evaluation Type=::NewEnumerator1)
      B=&FloatFromCurve(FloatCurve=DataInterface() {propertyValues {Curve {keys [
           <at 0 1.5 RCIM_Cubic>
           {arriveTangentWeight=0.0417, interpMode=RCIM_Cubic, leaveTangentWeight=0.0417, tangentMode=RCTM_User, time=0.1, value=0.5}
           {arriveTangentWeight=0.0333, interpMode=RCIM_Cubic, leaveTangentWeight=0.0333, tangentMode=RCTM_User, time=0.2, value=1.25}
           {arriveTangentWeight=0.0471, interpMode=RCIM_Cubic, leaveTangentWeight=0.0471, tangentMode=RCTM_User, time=0.3, value=0.35}
           {arriveTangentWeight=0.0333, interpMode=RCIM_Cubic, leaveTangentWeight=0.0333, tangentMode=RCTM_User, time=0.4, value=1}
           {arriveTangentWeight=0.0417, interpMode=RCIM_Cubic, leaveTangentWeight=0.0417, tangentMode=RCTM_User, time=0.5, value=0.25}
           {arriveTangentWeight=0.0417, interpMode=RCIM_Cubic, leaveTangentWeight=0.0417, tangentMode=RCTM_User, time=0.6, value=0.75}
           {arriveTangentWeight=0.0417, interpMode=RCIM_Cubic, leaveTangentWeight=0.0417, tangentMode=RCTM_User, time=0.7, value=0.1}
           {arriveTangentWeight=0.0333, interpMode=RCIM_Cubic, leaveTangentWeight=0.0333, tangentMode=RCTM_User, time=0.8, value=0.5}
           {arriveTangent=-2.5, interpMode=RCIM_Cubic, leaveTangent=-2.5, time=0.9, value=0.1}
           <at 1 0 RCIM_Cubic>
          ]}}})
     ))
    ScaleA=true
    ScaleRGB=true
   )
  ]
  rendererClasses [&NiagaraSpriteRendererProperties]
  renderers [{rendererClass=&NiagaraSpriteRendererProperties}]
 }
 TinkDust_Burst {
  emitterUpdateScript [
   &EmitterState(Life Cycle Mode=::NewEnumerator1, Loop Behavior=::NewEnumerator1, Loop Duration=0.1, UseLoopDelay=false)
   &SpawnBurst_Instantaneous(
    Spawn Count=&RandomRangeInt(Minimum=500, Maximum=250)
    Use Spawn Probability=false
    UseLoopCountLimit=false
   )
  ]
  particleSpawnScript [
   &InitializeParticle(
    Color=&self:SelectColorCurve(
     CurveSelect:NiagaraFloat=$User.Ranked_Curve
     ColorSelect:NiagaraFloat=&RandomRangeFloat(Minimum=0, Maximum=1)
     UnrankedColorCurve:NiagaraDataInterfaceColorCurve=DataInterface() {propertyValues {
       AlphaCurve {keys [<at 0 1>,<at 1 1>]}
       BlueCurve {keys [
         {interpMode=RCIM_Cubic, leaveTangentWeight=0.4171, tangentMode=RCTM_User, time=0, value=0.2831}
         {arriveTangent=1.9677, interpMode=RCIM_Cubic, leaveTangent=1.9677, tangentMode=RCTM_User, time=1, value=1}
        ]}
       GreenCurve {keys [
         {interpMode=RCIM_Cubic, leaveTangentWeight=0.4171, tangentMode=RCTM_User, time=0, value=0.2831}
         {arriveTangent=1.9677, interpMode=RCIM_Cubic, leaveTangent=1.9677, tangentMode=RCTM_User, time=1, value=1}
        ]}
       RedCurve {keys [
         {interpMode=RCIM_Cubic, leaveTangentWeight=0.4171, tangentMode=RCTM_User, time=0, value=0.2831}
         {arriveTangent=1.9677, interpMode=RCIM_Cubic, leaveTangent=1.9677, tangentMode=RCTM_User, time=1, value=1}
        ]}
      }}
     BronzeColorCurve:NiagaraDataInterfaceColorCurve=DataInterface() {propertyValues {
       AlphaCurve {keys [<at 0 1>,<at 1 1>]}
       BlueCurve {keys [<at 0 0.0185>,<at 1 0.1946>]}
       GreenCurve {keys [<at 0 0.0714>,<at 1 0.3557>]}
       RedCurve {keys [<at 0 0.2831>,<at 1 1>]}
      }}
     SilverColorCurve:NiagaraDataInterfaceColorCurve=DataInterface() {propertyValues {
       AlphaCurve {keys [<at 0 1>,<at 1 1>]}
       BlueCurve {keys [<at 0 0.0409>,<at 1 0.9734>]}
       GreenCurve {keys [<at 0 0.0273>,<at 1 0.9301>]}
       RedCurve {keys [<at 0 0.0194>,<at 1 0.8796>]}
      }}
     GoldColorCurve:NiagaraDataInterfaceColorCurve=DataInterface() {propertyValues {
       AlphaCurve {keys [<at 0 1>,<at 1 1>]}
       BlueCurve {keys [<at 0 0>,<at 1 0.2961>]}
       GreenCurve {keys [<at 0 0.1694>,<at 1 0.7654>]}
       RedCurve {keys [<at 0 0.2542>,<at 1 1>]}
      }}
     PlatColorCurve:NiagaraDataInterfaceColorCurve=DataInterface() {propertyValues {
       AlphaCurve {keys [<at 0 1>,<at 1 1>]}
       BlueCurve {keys [<at 0 0.0221>,<at 1 0.404>]}
       GreenCurve {keys [<at 0 0.05>,<at 1 0.9216>]}
       RedCurve {keys [<at 0 0.0021>,<at 1 0.0343>]}
      }}
     DiamondColorCurve:NiagaraDataInterfaceColorCurve=DataInterface() {propertyValues {
       AlphaCurve {keys [<at 0 1>,<at 1 1>]}
       BlueCurve {keys [<at 0 0.4>,<at 1 1>]}
       GreenCurve {keys [<at 0 0.2356>,<at 1 0.5833>]}
       RedCurve {keys [<at 0 0.0054>,<at 1 0>]}
      }}
     EliteColorCurve:NiagaraDataInterfaceColorCurve=DataInterface() {propertyValues {
       AlphaCurve {keys [<at 0 1>,<at 1 1>]}
       BlueCurve {keys [<at 0 0.0005 RCIM_Cubic>, {arriveTangent=0.3468, interpMode=RCIM_Cubic, leaveTangent=0.3468, tangentMode=RCTM_User, time=1, value=0.1}]}
       GreenCurve {keys [<at 0 0.0005 RCIM_Cubic>, {arriveTangent=0.3468, interpMode=RCIM_Cubic, leaveTangent=0.3468, tangentMode=RCTM_User, time=1, value=0.1}]}
       RedCurve {keys [<at 0 0.0005 RCIM_Cubic>, {arriveTangent=0.3468, interpMode=RCIM_Cubic, leaveTangent=0.3468, tangentMode=RCTM_User, time=1, value=0.1}]}
      }}
     ChampionColorCurve:NiagaraDataInterfaceColorCurve=DataInterface() {propertyValues {
       AlphaCurve {keys [<at 0 1>,<at 1 1>]}
       BlueCurve {keys [<at 0 0>,<at 0.5 0>,<at 1 0.0012>]}
       GreenCurve {keys [<at 0 0 RCIM_Cubic>, {arriveTangent=0.6654, interpMode=RCIM_Cubic, leaveTangent=0.6654, time=0.5, value=0.0908}, <at 1 0.6654 RCIM_Cubic>]}
       RedCurve {keys [
         <at 0 0.15 RCIM_Cubic>
         {arriveTangentWeight=0.2187, interpMode=RCIM_Cubic, leaveTangentWeight=0.2187, tangentMode=RCTM_User, time=0.5, value=1}
         <at 1 1 RCIM_Cubic>
        ]}
      }}
     UnrealColorCurve:NiagaraDataInterfaceColorCurve=DataInterface() {propertyValues {
       AlphaCurve {keys [<at 0 1>,<at 1 1>]}
       BlueCurve {keys [<at 0 0.2307 RCIM_Cubic>, {interpMode=RCIM_Cubic, leaveTangent=0.5013, time=1, value=0.2664}]}
       GreenCurve {keys [<at 0 0.0252 RCIM_Cubic>, {interpMode=RCIM_Cubic, leaveTangent=0.6649, time=1, value=0.0782}]}
       RedCurve {keys [
         <at 0 0.0222 RCIM_Cubic>
         {arriveTangentWeight=0.2187, interpMode=RCIM_Cubic, leaveTangent=0.4814, leaveTangentWeight=0.2187, time=1, value=0.3564}
        ]}
      }}
    )
    Lifetime Max=1.5
    Lifetime Min=0.5
    Lifetime Mode=::NewEnumerator1
    Sprite Size Mode=::NewEnumerator1
    Uniform Sprite Size Max=14
    Uniform Sprite Size Min=2
    UsePositionOffset=false
    Write Material Random=true
   )
   &ShapeLocation(MaskSpawnGroup=false, Override Seed=false, Sphere Radius=2, UseOwnerScale=true)
   &PointForce(
    Force Strength=&RandomRangeFloat(Minimum=150, Maximum=300)
    Module.Use Falloff=false
    Module.UseLinearForce=false
   )
   &AddVelocity(Velocity=<0 0 325>)
   &ApplyInitialForces()
  ]
  particleUpdateScript [
   &ParticleState()
   &ScaleSpriteSize(
    Uniform Curve Scale=&FloatFromCurve(FloatCurve=DataInterface() {propertyValues {Curve {keys [<at 0 1 RCIM_Cubic>, {arriveTangent=-2.2048, interpMode=RCIM_Cubic, leaveTangent=-2.2048, tangentMode=RCTM_User, time=1, value=0}]}}})
    Uniform Curve Sprite Scale=DataInterface() {propertyValues {Curve {keys [
        <at 0 1 RCIM_Cubic>
        {arriveTangentWeight=0.0462, interpMode=RCIM_Cubic, leaveTangentWeight=0.0449, tangentMode=RCTM_User, time=0.1354, value=0.2886}
        {arriveTangentWeight=0.0512, interpMode=RCIM_Cubic, leaveTangentWeight=0.0563, tangentMode=RCTM_User, time=0.2671, value=0.9417}
        {arriveTangentWeight=0.0484, interpMode=RCIM_Cubic, leaveTangentWeight=0.0536, tangentMode=RCTM_User, time=0.412, value=0.1231}
        {arriveTangentWeight=0.0535, interpMode=RCIM_Cubic, leaveTangentWeight=0.0475, tangentMode=RCTM_User, time=0.5725, value=0.9591}
        {arriveTangentWeight=0.0479, interpMode=RCIM_Cubic, leaveTangentWeight=0.0524, tangentMode=RCTM_User, time=0.7149, value=0.1294}
        {arriveTangentWeight=0.0536, interpMode=RCIM_Cubic, leaveTangentWeight=0.0443, tangentMode=RCTM_User, time=0.8709, value=0.9199}
        {arriveTangentWeight=0.043, interpMode=RCIM_Cubic, tangentMode=RCTM_User, time=1, value=0.2}
       ]}}}
   )
   &AccelerationForce(Acceleration=<0 0 -500>)
   &VectorNoiseForce(Force Amount=500, Module.Use Random Noise Time Interval=false)
   &SolveForcesAndVelocity(Clamp Velocity=false, Limit Acceleration=false)
   &ScaleColor(
    Scale RGB=&VectorFromFloat(Value=&Multiply_Float(
      A=&RandomRangeFloat(Minimum=2, Maximum=8, Evaluation Type=::NewEnumerator1)
      B=&FloatFromCurve(FloatCurve=DataInterface() {propertyValues {Curve {keys [
           <at 0 1.5 RCIM_Cubic>
           {arriveTangentWeight=0.0417, interpMode=RCIM_Cubic, leaveTangentWeight=0.0417, tangentMode=RCTM_User, time=0.1, value=0.5}
           {arriveTangentWeight=0.0333, interpMode=RCIM_Cubic, leaveTangentWeight=0.0333, tangentMode=RCTM_User, time=0.2, value=1.25}
           {arriveTangentWeight=0.0471, interpMode=RCIM_Cubic, leaveTangentWeight=0.0471, tangentMode=RCTM_User, time=0.3, value=0.35}
           {arriveTangentWeight=0.0333, interpMode=RCIM_Cubic, leaveTangentWeight=0.0333, tangentMode=RCTM_User, time=0.4, value=1}
           {arriveTangentWeight=0.0417, interpMode=RCIM_Cubic, leaveTangentWeight=0.0417, tangentMode=RCTM_User, time=0.5, value=0.25}
           {arriveTangentWeight=0.0417, interpMode=RCIM_Cubic, leaveTangentWeight=0.0417, tangentMode=RCTM_User, time=0.6, value=0.75}
           {arriveTangentWeight=0.0417, interpMode=RCIM_Cubic, leaveTangentWeight=0.0417, tangentMode=RCTM_User, time=0.7, value=0.1}
           {arriveTangentWeight=0.0333, interpMode=RCIM_Cubic, leaveTangentWeight=0.0333, tangentMode=RCTM_User, time=0.8, value=0.5}
           {arriveTangent=-2.5, interpMode=RCIM_Cubic, leaveTangent=-2.5, time=0.9, value=0.1}
           <at 1 0 RCIM_Cubic>
          ]}}})
     ))
    ScaleA=true
    ScaleRGB=true
   )
  ]
  rendererClasses [&NiagaraSpriteRendererProperties]
  renderers [{rendererClass=&NiagaraSpriteRendererProperties}]
 }
 Tinkerbell {
  emitterUpdateScript [
   &EmitterState()
   &SpawnBurst_Instantaneous(Use Spawn Probability=false, UseLoopCountLimit=false)
   &SpawnBurst_Instantaneous(Spawn Group=1, Use Spawn Probability=false, UseLoopCountLimit=false) {module=SpawnBurst_Instantaneous001}
  ]
  particleSpawnScript [&InitializeParticle(
    Color=&MaskLinearColorBySpawnGroup(
     Masked Linear Color=&self:SelectColorCurve(
      CurveSelect:NiagaraFloat=$User.Ranked_Curve
      ColorSelect:NiagaraFloat=0
      UnrankedColorCurve:NiagaraDataInterfaceColorCurve=DataInterface() {propertyValues {AlphaCurve {keys []}, BlueCurve {keys []}, GreenCurve {keys []}, RedCurve {keys []}}}
      BronzeColorCurve:NiagaraDataInterfaceColorCurve=DataInterface() {propertyValues {
        AlphaCurve {keys [<at 0 1>,<at 1 1>]}
        BlueCurve {keys [<at 0 0.0185>,<at 1 0.1946>]}
        GreenCurve {keys [<at 0 0.0714>,<at 1 0.3557>]}
        RedCurve {keys [<at 0 0.2831>,<at 1 1>]}
       }}
      SilverColorCurve:NiagaraDataInterfaceColorCurve=DataInterface() {propertyValues {
        AlphaCurve {keys [<at 0 1>,<at 1 1>]}
        BlueCurve {keys [<at 0 0.0409>,<at 1 0.9734>]}
        GreenCurve {keys [<at 0 0.0273>,<at 1 0.9301>]}
        RedCurve {keys [<at 0 0.0194>,<at 1 0.8796>]}
       }}
      GoldColorCurve:NiagaraDataInterfaceColorCurve=DataInterface() {propertyValues {
        AlphaCurve {keys [<at 0 1>,<at 1 1>]}
        BlueCurve {keys [<at 0 0>,<at 1 0.2961>]}
        GreenCurve {keys [<at 0 0.1694>,<at 1 0.7654>]}
        RedCurve {keys [<at 0 0.2542>,<at 1 1>]}
       }}
      PlatColorCurve:NiagaraDataInterfaceColorCurve=DataInterface() {propertyValues {
        AlphaCurve {keys [<at 0 1>,<at 1 1>]}
        BlueCurve {keys [<at 0 0.0021>,<at 1 0.0343>]}
        GreenCurve {keys [<at 0 0.05>,<at 1 0.9216>]}
        RedCurve {keys [<at 0 0.0101>,<at 1 0.1822>]}
       }}
      DiamondColorCurve:NiagaraDataInterfaceColorCurve=DataInterface() {propertyValues {
        AlphaCurve {keys [<at 0 1>,<at 1 1>]}
        BlueCurve {keys [<at 0 0.5>,<at 1 1>]}
        GreenCurve {keys [<at 0 0.3356>,<at 1 0.8333>]}
        RedCurve {keys [<at 0 0.0068>,<at 1 0.5>]}
       }}
      EliteColorCurve:NiagaraDataInterfaceColorCurve=DataInterface() {propertyValues {
        AlphaCurve {keys [<at 0 1>,<at 1 1>]}
        BlueCurve {keys [<at 0 0 RCIM_Cubic>, {arriveTangent=2.8442, interpMode=RCIM_Cubic, leaveTangent=2.8442, tangentMode=RCTM_User, time=1, value=1}]}
        GreenCurve {keys [<at 0 0 RCIM_Cubic>, {arriveTangent=2.8442, interpMode=RCIM_Cubic, leaveTangent=2.8442, tangentMode=RCTM_User, time=1, value=1}]}
        RedCurve {keys [<at 0 0 RCIM_Cubic>, {arriveTangent=2.8442, interpMode=RCIM_Cubic, leaveTangent=2.8442, tangentMode=RCTM_User, time=1, value=1}]}
       }}
      ChampionColorCurve:NiagaraDataInterfaceColorCurve=DataInterface() {propertyValues {
        AlphaCurve {keys [<at 0 1>,<at 1 1>]}
        BlueCurve {keys [<at 0 0>,<at 0.5 0>,<at 1 0.0012>]}
        GreenCurve {keys [<at 0 0 RCIM_Cubic>, {arriveTangent=0.6654, interpMode=RCIM_Cubic, leaveTangent=0.6654, time=0.5, value=0.0908}, <at 1 0.6654 RCIM_Cubic>]}
        RedCurve {keys [
          <at 0 0.15 RCIM_Cubic>
          {arriveTangentWeight=0.2187, interpMode=RCIM_Cubic, leaveTangentWeight=0.2187, tangentMode=RCTM_User, time=0.5, value=1}
          <at 1 1 RCIM_Cubic>
         ]}
       }}
      UnrealColorCurve:NiagaraDataInterfaceColorCurve=DataInterface() {propertyValues {
        AlphaCurve {keys [<at 0 1>,<at 1 1>]}
        BlueCurve {keys [
          <at 0 0.2307 RCIM_Cubic>
          {arriveTangent=0.5013, interpMode=RCIM_Cubic, leaveTangent=0.5013, time=0.5, value=0.2664}
          <at 1 0.732 RCIM_Cubic>
         ]}
        GreenCurve {keys [
          <at 0 0.0252 RCIM_Cubic>
          {arriveTangent=0.6649, interpMode=RCIM_Cubic, leaveTangent=0.6649, time=0.5, value=0.0782}
          <at 1 0.6901 RCIM_Cubic>
         ]}
        RedCurve {keys [
          <at 0 0.0222 RCIM_Cubic>
          {
           arriveTangent=0.4814
           arriveTangentWeight=0.2187
           interpMode=RCIM_Cubic
           leaveTangent=0.4814
           leaveTangentWeight=0.2187
           time=0.5
           value=0.3564
          }
          <at 1 0.5036 RCIM_Cubic>
         ]}
       }}
     )
     Passthrough Linear Color=&self:SelectColorCurve(
      CurveSelect:NiagaraFloat=$User.Ranked_Curve
      ColorSelect:NiagaraFloat=1
      UnrankedColorCurve:NiagaraDataInterfaceColorCurve=DataInterface() {propertyValues {
        AlphaCurve {keys [<at 0 1>,<at 1 1>]}
        BlueCurve {keys [
          {interpMode=RCIM_Cubic, leaveTangentWeight=0.4171, tangentMode=RCTM_User, time=0, value=0.2831}
          {arriveTangent=1.9677, interpMode=RCIM_Cubic, leaveTangent=1.9677, tangentMode=RCTM_User, time=1, value=1}
         ]}
        GreenCurve {keys [
          {interpMode=RCIM_Cubic, leaveTangentWeight=0.4171, tangentMode=RCTM_User, time=0, value=0.2831}
          {arriveTangent=1.9677, interpMode=RCIM_Cubic, leaveTangent=1.9677, tangentMode=RCTM_User, time=1, value=1}
         ]}
        RedCurve {keys [
          {interpMode=RCIM_Cubic, leaveTangentWeight=0.4171, tangentMode=RCTM_User, time=0, value=0.2831}
          {arriveTangent=1.9677, interpMode=RCIM_Cubic, leaveTangent=1.9677, tangentMode=RCTM_User, time=1, value=1}
         ]}
       }}
      BronzeColorCurve:NiagaraDataInterfaceColorCurve=DataInterface() {propertyValues {
        AlphaCurve {keys [<at 0 1>,<at 1 1>]}
        BlueCurve {keys [<at 0 0.0185>,<at 1 0.1946>]}
        GreenCurve {keys [<at 0 0.0714>,<at 1 0.3557>]}
        RedCurve {keys [<at 0 0.2831>,<at 1 1>]}
       }}
      SilverColorCurve:NiagaraDataInterfaceColorCurve=DataInterface() {propertyValues {
        AlphaCurve {keys [<at 0 1>,<at 1 1>]}
        BlueCurve {keys [<at 0 0.0409>,<at 1 0.9734>]}
        GreenCurve {keys [<at 0 0.0273>,<at 1 0.9301>]}
        RedCurve {keys [<at 0 0.0194>,<at 1 0.8796>]}
       }}
      GoldColorCurve:NiagaraDataInterfaceColorCurve=DataInterface() {propertyValues {
        AlphaCurve {keys [<at 0 1>,<at 1 1>]}
        BlueCurve {keys [<at 0 0>,<at 1 0.2961>]}
        GreenCurve {keys [<at 0 0.1694>,<at 1 0.7654>]}
        RedCurve {keys [<at 0 0.2542>,<at 1 1>]}
       }}
      PlatColorCurve:NiagaraDataInterfaceColorCurve=DataInterface() {propertyValues {
        AlphaCurve {keys [<at 0 1>,<at 1 1>]}
        BlueCurve {keys [<at 0 0.0221>,<at 1 0.404>]}
        GreenCurve {keys [<at 0 0.05>,<at 1 0.9216>]}
        RedCurve {keys [<at 0 0.0021>,<at 1 0.0343>]}
       }}
      DiamondColorCurve:NiagaraDataInterfaceColorCurve=DataInterface() {propertyValues {
        AlphaCurve {keys [<at 0 1>,<at 1 1>]}
        BlueCurve {keys [<at 0 0.4>,<at 1 1>]}
        GreenCurve {keys [<at 0 0.2356>,<at 1 0.5833>]}
        RedCurve {keys [<at 0 0.0054>,<at 1 0>]}
       }}
      EliteColorCurve:NiagaraDataInterfaceColorCurve=DataInterface() {propertyValues {
        AlphaCurve {keys [<at 0 1>,<at 1 1>]}
        BlueCurve {keys [<at 0 1 RCIM_Cubic>]}
        GreenCurve {keys [<at 0 1 RCIM_Cubic>]}
        RedCurve {keys [<at 0 1 RCIM_Cubic>]}
       }}
      ChampionColorCurve:NiagaraDataInterfaceColorCurve=DataInterface() {propertyValues {
        AlphaCurve {keys [<at 0 1>,<at 1 1>]}
        BlueCurve {keys [<at 0 0>,<at 0.5 0>,<at 1 0.0012>]}
        GreenCurve {keys [<at 0 0 RCIM_Cubic>, {arriveTangent=0.6654, interpMode=RCIM_Cubic, leaveTangent=0.6654, time=0.5, value=0.0908}, <at 1 0.6654 RCIM_Cubic>]}
        RedCurve {keys [
          <at 0 0.15 RCIM_Cubic>
          {arriveTangentWeight=0.2187, interpMode=RCIM_Cubic, leaveTangentWeight=0.2187, tangentMode=RCTM_User, time=0.5, value=1}
          <at 1 1 RCIM_Cubic>
         ]}
       }}
      UnrealColorCurve:NiagaraDataInterfaceColorCurve=DataInterface() {propertyValues {
        AlphaCurve {keys [<at 0 1>,<at 1 1>]}
        BlueCurve {keys [<at 0 0.2307 RCIM_Cubic>, {interpMode=RCIM_Cubic, leaveTangent=0.5013, time=1, value=0.2664}]}
        GreenCurve {keys [<at 0 0.0252 RCIM_Cubic>, {interpMode=RCIM_Cubic, leaveTangent=0.6649, time=1, value=0.0782}]}
        RedCurve {keys [
          <at 0 0.0222 RCIM_Cubic>
          {arriveTangentWeight=0.2187, interpMode=RCIM_Cubic, leaveTangent=0.4814, leaveTangentWeight=0.2187, time=1, value=0.3564}
         ]}
       }}
     )
    )
    Lifetime=1.82
    Sprite Rotation Mode=::NewEnumerator2
    Sprite Size Mode=::NewEnumerator3
    Uniform Sprite Size=20
    UsePositionOffset=false
    Write Material Random=true
   )]
  particleUpdateScript [
   &ParticleState()
   &SpriteRotationRate(Rotation Rate=&MaskFloatBySpawnGroup(Masked Float=45, Passthrough Float=-45))
   &ScaleSpriteSize(
    Uniform Curve Index=$Emitter.NormalizedLoopAge
    Uniform Curve Sprite Scale=DataInterface() {propertyValues {Curve {keys [{arriveTangent=101.8832, interpMode=RCIM_Cubic, leaveTangent=101.8832, tangentMode=RCTM_User, time=0, value=0}, <at 0.3 1 RCIM_Cubic>]}}}
   )
   &ScaleColor(
    Scale Alpha=&FloatFromCurve(
     FloatCurve=DataInterface() {propertyValues {Curve {keys [{arriveTangent=15.6232, interpMode=RCIM_Cubic, leaveTangent=15.6232, tangentMode=RCTM_User, time=0, value=0}, <at 0.2 1 RCIM_Cubic>]}}}
     CurveIndex=$Emitter.LoopedAge
    )
    ScaleA=true
    ScaleRGB=true
   )
   &DynamicMaterialParameters(
    Index 0 Param 1=1
    Index 0 Param 2=1
    Index 0 Param 3=1
    Module.Param0WriteEnabled=true
    Module.Param1WriteEnabled=true
    Module.Param2WriteEnabled=true
    Module.Param3WriteEnabled=true
   )
   &InheritSourceMovement(ApplyForce=false, ApplyPosition=true, ApplyPositionFalloff=false)
   &SolveForcesAndVelocity(Clamp Velocity=false, Limit Acceleration=false)
  ]
  rendererClasses [&NiagaraSpriteRendererProperties]
  renderers [{rendererClass=&NiagaraSpriteRendererProperties}]
 }
}
summary {
 systemName=NS_RankedProgression_Ember
 emitters [
  {emitterName=Tinkerbell, rendererClasses [&NiagaraSpriteRendererProperties]}
  {emitterName=Corona_Burst, rendererClasses [&NiagaraSpriteRendererProperties]}
  {emitterName=TinkDust, rendererClasses [&NiagaraSpriteRendererProperties]}
  {emitterName=TinkDust_Burst, rendererClasses [&NiagaraSpriteRendererProperties]}
 ]
 userVariables [User.Ranked_Curve {type=NiagaraFloat, defaultValue=0, description=""}]
}
