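# Niagara flare-up effect for ranked progression, with fire sprites, flame meshes, embers, a corona and flame decals.
# Continuous fire peaks in timed bursts near two seconds, then tapers out by about four seconds via emitter-age curves.
# User.Unranked swaps fire and embers for short stroke bursts, while User.Ranked_Curve gates the corona and selects tier palettes through SelectColorCurve.
# Flames use randomized SubUV flipbooks with dynamic material parameters.
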NiagaraSystem /FNGameplayCues/Effects/Fort_Effects/Effects/Characters/Generic/RankedProgression/NS_RankedProgression_FlareUp_v3
dependencies {
 usedDataInterfaces [
  &NiagaraDataInterfaceCurve
  &NiagaraDataInterfaceMeshRendererInfo
  &NiagaraDataInterfaceVector2DCurve
  &NiagaraDataInterfaceSpriteRendererInfo
  &NiagaraDataInterfaceColorCurve
 ]
 usedDynamicInputs []
 usedModules [
  &SystemState
  &self:SystemSpawnScript.SystemScriptSource.SystemScriptGraph.NiagaraNodeAssignment_0.SetVariables_#1
 ]
}
emitters {
 BottomFireSprites {
  emitterUpdateScript [
   &EmitterState()
   &SpawnRate(
    SpawnRate=&MakeCustomFloatFromBool(
     Bool=$User.Unranked
     True Float=0
     False Float=&Multiply_Float(
      A=&RandomRangeFloat(Minimum=10, Maximum=16)
      B=&FloatFromCurve(FloatCurve=DataInterface() {propertyValues {Curve {keys [<at 3.257 1>,<at 4 0>]}}}, CurveIndex=$Emitter.Age)
     )
    )
    Use Spawn Probability=false
   )
  ]
  particleSpawnScript [
   &InitializeParticle(
    Lifetime Max=1
    Lifetime Min=0.5
    Lifetime Mode=::NewEnumerator1
    Sprite Size Mode=::NewEnumerator1
    Sprite UV Mode=::NewEnumerator1
    Uniform Sprite Size Max=35
    Uniform Sprite Size Min=25
    UsePositionOffset=false
    Write Material Random=true
   )
   &CameraOffset(Camera Offset Amount=&RandomRangeFloat(Minimum=5, Maximum=25))
  ]
  particleUpdateScript [
   &ParticleState()
   &ScaleSpriteSize(
    Non-Uniform Curve Sprite Scale=DataInterface() {propertyValues {
      XCurve {keys [{arriveTangent=2.4273, interpMode=RCIM_Cubic, leaveTangent=2.4273, tangentMode=RCTM_User, time=0, value=0.25}, <at 1 1 RCIM_Cubic>]}
      YCurve {keys [{arriveTangent=2.4273, interpMode=RCIM_Cubic, leaveTangent=2.4273, tangentMode=RCTM_User, time=0, value=0.25}, <at 1 1 RCIM_Cubic>]}
     }}
    Scale Sprite Size Mode=::NewEnumerator3
   )
   &SubUVAnimation(
    "Sprite Renderer "=DataInterface() {propertyValues {SpriteRenderer=&self:BottomFireSprites_9.NiagaraSpriteRendererProperties_0}}
    Start Frame Range Override=&RandomRangeInt(Minimum=0, Maximum=10)
    SubUV Lookup Index Scale:NiagaraFloat=&RandomRangeFloat(Minimum=0.9, Maximum=1.1)
    UseEndFrame=true
    UseStartFrame=true
   )
   &DynamicMaterialParameters(
    Index 0 Param 1=&Multiply_Float(
     A=&RandomRangeFloat(Minimum=2, Maximum=3)
     B=&FloatFromCurve(FloatCurve=DataInterface() {propertyValues {Curve {keys [<at 0 1 RCIM_Cubic>, {arriveTangent=-2.109, interpMode=RCIM_Cubic, leaveTangent=-2.109, tangentMode=RCTM_User, time=1, value=0.1}]}}})
    )
    Index 0 Param 2=&RandomRangeFloat(Minimum=0, Maximum=1)
    Index 0 Param 3=8
    Index 0 Param 4=$Particles.SubImageIndex
    Index 1 Param 1=&RandomRangeFloat(Minimum=-1, Maximum=1)
    Index 1 Param 2=&RandomRangeFloat(Minimum=0, Maximum=3)
    Index 1 Param 3=&FloatFromCurve(FloatCurve=DataInterface() {propertyValues {Curve {keys [
         {arriveTangent=4.1785, interpMode=RCIM_Cubic, leaveTangent=4.1785, tangentMode=RCTM_User, time=0, value=0}
         {arriveTangent=0.9995, interpMode=RCIM_Cubic, leaveTangent=0.9995, tangentMode=RCTM_User, time=0.2223, value=0.6825}
         {arriveTangent=-0.0405, interpMode=RCIM_Cubic, leaveTangent=-0.0405, tangentMode=RCTM_User, time=0.7, value=0.9}
         <at 1 0.25>
        ]}}})
    Index 1 Param 4=1
    Module.Param0WriteEnabled=true
    Module.Param0WriteEnabled1=true
    Module.Param1WriteEnabled=true
    Module.Param1WriteEnabled1=true
    Module.Param2WriteEnabled=true
    Module.Param2WriteEnabled1=true
    Module.Param3WriteEnabled=true
    Module.Param3WriteEnabled1=true
    Write Parameter Index 1=true
   )
   &ScaleColor(
    Scale Alpha=&FloatFromCurve(
     FloatCurve=DataInterface() {propertyValues {Curve {keys [
         {interpMode=RCIM_Cubic, tangentMode=RCTM_User, time=0, value=0}
         {interpMode=RCIM_Cubic, tangentMode=RCTM_User, time=0.35, value=1}
         {arriveTangentWeight=0.3727, interpMode=RCIM_Cubic, tangentMode=RCTM_User, time=1, value=0}
        ]}}}
     Scale Curve=&FloatFromCurve(
      FloatCurve=DataInterface() {propertyValues {Curve {keys [{arriveTangent=-2.4644, interpMode=RCIM_Cubic, leaveTangent=-2.4644, tangentMode=RCTM_User, time=3.257, value=1}, <at 4 0 RCIM_Cubic>]}}}
      CurveIndex=$Emitter.Age
     )
    )
    ScaleA=true
    ScaleRGB=true
   )
   &AccelerationForce(Acceleration=&RandomRangeVector(Minimum=<0 0 75>, Maximum=<1 1 150>))
   &SolveForcesAndVelocity(Clamp Velocity=false, Limit Acceleration=false)
  ]
  rendererClasses [&NiagaraSpriteRendererProperties]
  renderers [{rendererClass=&NiagaraSpriteRendererProperties}]
 }
 Core_Fire_Sprites {
  emitterUpdateScript [
   &EmitterState()
   &SpawnBurst_Instantaneous(
    Spawn Count=&MakeCustomIntFromBool(Bool=$User.Unranked, True INT=0, False INT=20)
    Spawn Time=2.0117
    Use Spawn Probability=false
    UseLoopCountLimit=false
   )
   &SpawnRate(
    SpawnRate=&MakeCustomFloatFromBool(
     Bool=$User.Unranked
     True Float=0
     False Float=&Multiply_Float(
      A=&RandomRangeFloat(Minimum=10, Maximum=16)
      B=&FloatFromCurve(FloatCurve=DataInterface() {propertyValues {Curve {keys [<at 3.257 1>,<at 4 0>]}}}, CurveIndex=$Emitter.Age)
     )
    )
    Use Spawn Probability=false
   )
  ]
  particleSpawnScript [
   &InitializeParticle(
    Color=<rgba 1 1 1 0.5>
    Lifetime Max=1
    Lifetime Min=0.5
    Lifetime Mode=::NewEnumerator1
    Sprite Size Mode=::NewEnumerator3
    Sprite UV Mode=::NewEnumerator1
    Uniform Sprite Size=&Multiply_Float(A=&RandomRangeFloat(Minimum=30, Maximum=40), B=$System.NormalizedFirstBurst)
    UsePositionOffset=true
    Write Material Random=true
   )
   &CameraOffset(Camera Offset Amount=&RandomRangeFloat(Minimum=-10, Maximum=20))
  ]
  particleUpdateScript [
   &ParticleState()
   &ScaleSpriteSize(
    Non-Uniform Curve Sprite Scale=DataInterface() {propertyValues {
      XCurve {keys [{arriveTangent=2.0551, interpMode=RCIM_Cubic, leaveTangent=2.0551, tangentMode=RCTM_User, time=0, value=0.25}, <at 1 1 RCIM_Cubic>]}
      YCurve {keys [{tangentMode=RCTM_User, time=0, value=1}, {tangentMode=RCTM_User, time=1, value=1.5}]}
     }}
    Scale Sprite Size Mode=::NewEnumerator3
   )
   &SubUVAnimation(
    "Sprite Renderer "=DataInterface() {propertyValues {SpriteRenderer=&self:Core_Fire_Sprites_2.NiagaraSpriteRendererProperties_0}}
    Start Frame Range Override=&RandomRangeInt(Minimum=0, Maximum=10)
    SubUV Lookup Index Scale:NiagaraFloat=&RandomRangeFloat(Minimum=0.9, Maximum=1.1)
    UseEndFrame=true
    UseStartFrame=true
   )
   &DynamicMaterialParameters(
    Index 0 Param 1=&RandomRangeFloat(Minimum=1, Maximum=2)
    Index 0 Param 2=&RandomRangeFloat(Minimum=0, Maximum=1)
    Index 0 Param 3=8
    Index 0 Param 4=$Particles.SubImageIndex
    Index 1 Param 1=&RandomRangeFloat(Minimum=-1, Maximum=1)
    Index 1 Param 2=&RandomRangeFloat(Minimum=0, Maximum=3)
    Index 1 Param 3=&FloatFromCurve(FloatCurve=DataInterface() {propertyValues {Curve {keys [
         {arriveTangent=4.1785, interpMode=RCIM_Cubic, leaveTangent=4.1785, tangentMode=RCTM_User, time=0, value=0}
         {arriveTangent=0.9995, interpMode=RCIM_Cubic, leaveTangent=0.9995, tangentMode=RCTM_User, time=0.2223, value=0.6825}
         {arriveTangent=-0.0405, interpMode=RCIM_Cubic, leaveTangent=-0.0405, tangentMode=RCTM_User, time=0.7, value=0.9}
         <at 1 0.25>
        ]}}})
    Index 1 Param 4=1
    Module.Param0WriteEnabled=true
    Module.Param0WriteEnabled1=true
    Module.Param1WriteEnabled=true
    Module.Param1WriteEnabled1=true
    Module.Param2WriteEnabled=true
    Module.Param2WriteEnabled1=true
    Module.Param3WriteEnabled=true
    Module.Param3WriteEnabled1=true
    Write Parameter Index 1=true
   )
   &ScaleColor(
    Scale Alpha=&FloatFromCurve(
     FloatCurve=DataInterface() {propertyValues {Curve {keys [
         {interpMode=RCIM_Cubic, tangentMode=RCTM_User, time=0, value=0}
         {interpMode=RCIM_Cubic, tangentMode=RCTM_User, time=0.35, value=1}
         {arriveTangentWeight=0.3727, interpMode=RCIM_Cubic, tangentMode=RCTM_User, time=1, value=0}
        ]}}}
     Scale Curve=&FloatFromCurve(
      FloatCurve=DataInterface() {propertyValues {Curve {keys [{arriveTangent=-2.4644, interpMode=RCIM_Cubic, leaveTangent=-2.4644, tangentMode=RCTM_User, time=3.257, value=1}, <at 4 0 RCIM_Cubic>]}}}
      CurveIndex=$Emitter.Age
     )
    )
    ScaleA=true
    ScaleRGB=true
   )
   &AccelerationForce(Acceleration=<0 0 100>)
   &SolveForcesAndVelocity(Clamp Velocity=false, Limit Acceleration=false)
  ]
  rendererClasses [&NiagaraSpriteRendererProperties]
  renderers [{rendererClass=&NiagaraSpriteRendererProperties}]
 }
 Corona_Burst {
  emitterUpdateScript [
   &EmitterState()
   &SpawnBurst_Instantaneous(
    Spawn Count=&MakeCustomIntFromBool(
     Bool=&SetBoolByFloatComparison(A=$User.Ranked_Curve, B=5, Comparison Type=::NewEnumerator2)
     True INT=1
     False INT=0
    )
    Use Spawn Probability=false
    UseLoopCountLimit=false
   )
  ]
  particleSpawnScript [
   &InitializeParticle(
    Color=&self:SelectColorCurve(
     CurveSelect:NiagaraFloat=$User.Ranked_Curve
     ColorSelect:NiagaraFloat=0.5
     UnrankedColorCurve:NiagaraDataInterfaceColorCurve=DataInterface() {propertyValues {
       AlphaCurve {keys [<at 0 1>,<at 1 1>]}
       BlueCurve {keys [
         {interpMode=RCIM_Cubic, leaveTangentWeight=0.4171, tangentMode=RCTM_User, time=0, value=0.2831}
         {arriveTangent=1.9677, interpMode=RCIM_Cubic, leaveTangent=1.9677, tangentMode=RCTM_User, time=1, value=1}
        ]}
       GreenCurve {keys [
         {interpMode=RCIM_Cubic, leaveTangentWeight=0.4171, tangentMode=RCTM_User, time=0, value=0.2831}
         {arriveTangent=1.9677, interpMode=RCIM_Cubic, leaveTangent=1.9677, tangentMode=RCTM_User, time=1, value=1}
        ]}
       RedCurve {keys [
         {interpMode=RCIM_Cubic, leaveTangentWeight=0.4171, tangentMode=RCTM_User, time=0, value=0.2831}
         {arriveTangent=1.9677, interpMode=RCIM_Cubic, leaveTangent=1.9677, tangentMode=RCTM_User, time=1, value=1}
        ]}
      }}
     BronzeColorCurve:NiagaraDataInterfaceColorCurve=DataInterface() {propertyValues {
       AlphaCurve {keys [<at 0 1>,<at 1 1>]}
       BlueCurve {keys [<at 0 0.0185>,<at 1 0.1946>]}
       GreenCurve {keys [<at 0 0.0714>,<at 1 0.3557>]}
       RedCurve {keys [<at 0 0.2831>,<at 1 1>]}
      }}
     SilverColorCurve:NiagaraDataInterfaceColorCurve=DataInterface() {propertyValues {
       AlphaCurve {keys [<at 0 1>,<at 1 1>]}
       BlueCurve {keys [<at 0 0.0409>,<at 1 0.9734>]}
       GreenCurve {keys [<at 0 0.0273>,<at 1 0.9301>]}
       RedCurve {keys [<at 0 0.0194>,<at 1 0.8796>]}
      }}
     GoldColorCurve:NiagaraDataInterfaceColorCurve=DataInterface() {propertyValues {
       AlphaCurve {keys [<at 0 1>,<at 1 1>]}
       BlueCurve {keys [<at 0 0>,<at 1 0.2961>]}
       GreenCurve {keys [<at 0 0.1694>,<at 1 0.7654>]}
       RedCurve {keys [<at 0 0.2542>,<at 1 1>]}
      }}
     PlatColorCurve:NiagaraDataInterfaceColorCurve=DataInterface() {propertyValues {
       AlphaCurve {keys [<at 0 1>,<at 1 1>]}
       BlueCurve {keys [<at 0 0.0221>,<at 1 0.404>]}
       GreenCurve {keys [<at 0 0.05>,<at 1 0.9216>]}
       RedCurve {keys [<at 0 0.0021>,<at 1 0.0343>]}
      }}
     DiamondColorCurve:NiagaraDataInterfaceColorCurve=DataInterface() {propertyValues {
       AlphaCurve {keys [<at 0 1>,<at 1 1>]}
       BlueCurve {keys [<at 0 0.4>,<at 1 1>]}
       GreenCurve {keys [<at 0 0.2356>,<at 1 0.5833>]}
       RedCurve {keys [<at 0 0.0054>,<at 1 0>]}
      }}
     EliteColorCurve:NiagaraDataInterfaceColorCurve=DataInterface() {propertyValues {
       AlphaCurve {keys [<at 0 1>]}
       BlueCurve {keys [<at 0 1 RCIM_Cubic>]}
       GreenCurve {keys [<at 0 1 RCIM_Cubic>]}
       RedCurve {keys [<at 0 1 RCIM_Cubic>]}
      }}
     ChampionColorCurve:NiagaraDataInterfaceColorCurve=DataInterface() {propertyValues {
       AlphaCurve {keys [<at 0 1>,<at 1 1>]}
       BlueCurve {keys [<at 0 0>,<at 0.5 0>,<at 1 0.0012>]}
       GreenCurve {keys [<at 0 0 RCIM_Cubic>, {arriveTangent=0.6654, interpMode=RCIM_Cubic, leaveTangent=0.6654, time=0.5, value=0.0908}, <at 1 0.6654 RCIM_Cubic>]}
       RedCurve {keys [
         <at 0 0.15 RCIM_Cubic>
         {arriveTangentWeight=0.2187, interpMode=RCIM_Cubic, leaveTangentWeight=0.2187, tangentMode=RCTM_User, time=0.5, value=1}
         <at 1 1 RCIM_Cubic>
        ]}
      }}
     UnrealColorCurve:NiagaraDataInterfaceColorCurve=DataInterface() {propertyValues {
       AlphaCurve {keys [<at 0 10>]}
       BlueCurve {keys [{interpMode=RCIM_Cubic, leaveTangent=0.5013, time=1, value=0.2664}]}
       GreenCurve {keys [{interpMode=RCIM_Cubic, leaveTangent=0.6649, time=1, value=0.0782}]}
       RedCurve {keys [{arriveTangentWeight=0.2187, interpMode=RCIM_Cubic, leaveTangent=0.4814, leaveTangentWeight=0.2187, time=1, value=0.3564}]}
      }}
    )
    Lifetime=4
    Position Offset=<-5 0 0>
    Position Offset Coordinate Space=::Local
    Sprite Rotation Mode=::NewEnumerator2
    Sprite Size Mode=::NewEnumerator3
    Uniform Sprite Size=100
    UsePositionOffset=true
    Write Material Random=true
   )
   &CameraOffset(Camera Offset Amount=25)
  ]
  particleUpdateScript [
   &ParticleState()
   &ScaleSpriteSize(
    Scale Sprite Size Mode=::NewEnumerator0
    Uniform Scale Factor=&Lerp_Float(
     A=0.5
     B=1
     Alpha=&FloatFromCurve(
      FloatCurve=DataInterface() {propertyValues {Curve {keys [{arriveTangent=-5.0707, interpMode=RCIM_Cubic, leaveTangent=-5.0707, tangentMode=RCTM_User, time=3.257, value=1}, <at 3.6 0 RCIM_Cubic>]}}}
      CurveIndex=$Emitter.Age
     )
    )
   )
   &DynamicMaterialParameters(
    Index 0 Param 1=&FloatFromCurve(
     FloatCurve=DataInterface() {propertyValues {Curve {keys [{arriveTangent=-6.5619, interpMode=RCIM_Cubic, leaveTangent=-6.5619, tangentMode=RCTM_User, time=3.257, value=1}, <at 3.6 0 RCIM_Cubic>]}}}
     CurveIndex=$Emitter.Age
    )
    Index 0 Param 2=1
    Index 0 Param 3=1
    Index 0 Param 4=1
    Index 1 Param 1=6.667
    Index 1 Param 2=&FloatFromCurve(
     FloatCurve=DataInterface() {propertyValues {Curve {keys [{arriveTangent=9.9946, interpMode=RCIM_Cubic, leaveTangent=9.9946, tangentMode=RCTM_User, time=3.3, value=0}, <at 3.6 1 RCIM_Cubic>]}}}
     CurveIndex=$Emitter.Age
    )
    Index 1 Param 3=1
    Module.Param0WriteEnabled=true
    Module.Param0WriteEnabled1=true
    Module.Param1WriteEnabled=true
    Module.Param1WriteEnabled1=true
    Module.Param2WriteEnabled=true
    Module.Param2WriteEnabled1=true
    Module.Param3WriteEnabled=true
    Module.Param3WriteEnabled1=true
    Write Parameter Index 1=true
   )
   &ScaleColor(
    Scale RGB=&VectorFromFloat(Value=&FloatFromCurve(
      FloatCurve=DataInterface() {propertyValues {Curve {keys [
          {arriveTangent=97.8879, interpMode=RCIM_Cubic, leaveTangent=97.8879, tangentMode=RCTM_User, time=1.9, value=1}
          <at 2.09 10 RCIM_Cubic>
          {
           arriveTangent=-138.6604
           arriveTangentWeight=2.0384
           interpMode=RCIM_Cubic
           leaveTangentWeight=19.5872
           tangentMode=RCTM_Break
           time=2.2
           value=1
          }
          {
           arriveTangentWeight=1.1743
           interpMode=RCIM_Cubic
           leaveTangent=28.2703
           leaveTangentWeight=0.2048
           tangentMode=RCTM_Break
           time=3.1
           value=1
          }
          {
           arriveTangentWeight=0.0782
           interpMode=RCIM_Cubic
           leaveTangent=-34.3158
           leaveTangentWeight=0.3702
           tangentMode=RCTM_Break
           time=3.257
           value=5
          }
          <at 3.6 0>
         ]}}}
      CurveIndex=$Emitter.Age
     ))
    ScaleA=true
    ScaleRGB=true
   )
  ]
  rendererClasses [&NiagaraSpriteRendererProperties]
  renderers [{rendererClass=&NiagaraSpriteRendererProperties}]
 }
 Embers {
  emitterUpdateScript [
   &EmitterState()
   &SpawnBurst_Instantaneous(
    Spawn Count=&MakeCustomIntFromBool(Bool=$User.Unranked, True INT=0, False INT=75)
    Spawn Time=2.05
    Use Spawn Probability=false
    UseLoopCountLimit=false
   )
   &SpawnRate(
    SpawnRate=&MakeCustomFloatFromBool(
     Bool=$User.Unranked
     True Float=0
     False Float=&Multiply_Float(
      A=&RandomRangeFloat(Minimum=0, Maximum=75, Evaluation Type=::NewEnumerator1)
      B=&FloatFromCurve(FloatCurve=DataInterface() {propertyValues {Curve {keys [<at 3.257 1>,<at 4 0>]}}}, CurveIndex=$Emitter.Age)
     )
    )
    Use Spawn Probability=false
   )
  ]
  particleSpawnScript [
   &InitializeParticle(
    Color=&MakeLinearColorFromVectorAndFloat(
     "Vector (RGB)"=&VectorFromFloat(Value=&RandomRangeFloat(Minimum=1, Maximum=3))
     "Float (Alpha)"=1
    )
    Lifetime Min=0.3
    Lifetime Mode=::NewEnumerator1
    Sprite Size=&Vector2DFromFloat(Value=&FloatFromCurve(
      FloatCurve=DataInterface() {propertyValues {Curve {keys [<at 0 1.2>,<at 0.1257 0.9847>,<at 0.2996 0.5>,<at 1 0.3>]}}}
      CurveIndex=&RandomRangeFloat(Minimum=0, Maximum=1)
      Scale Curve=3
     ))
    Sprite Size Mode=::NewEnumerator4
    UsePositionOffset=false
    Write Material Random=true
   )
   &ShapeLocation(Hemisphere Distribution=<1 0.5>, MaskSpawnGroup=false, Override Seed=false, Sphere Radius=8, UseOwnerScale=true)
   &AddVelocity(
    Module.Use Falloff Distance=false
    Override Seed=false
    Velocity Mode=::NewEnumerator1
    Velocity Speed=&RandomRangeFloat(Minimum=25, Maximum=50)
   )
  ]
  particleUpdateScript [
   &ParticleState()
   &AccelerationForce(Acceleration=&RandomRangeVector(Minimum=<0 0 150>, Maximum=<0 0 250>))
   /Niagara/Modules/Update/Forces/CurlNoiseForce(
    Noise Frequency=0.04
    Noise Quality / Cost=::NewEnumerator0
    Noise Strength=&FloatFromCurve(FloatCurve=DataInterface() {propertyValues {Curve {keys [<at 0 1>,<at 1 0>]}}}, Scale Curve=500)
    Pan Noise=true
    Pan Noise Field=<0 0 1>
   )
   &Drag(
    Drag=&FloatFromCurve(FloatCurve=DataInterface() {propertyValues {Curve {keys [<at 0 0.5>,<at 0.5 0.5>,<at 1 1>]}}}, Scale Curve=9)
    Use Linear Drag=true
    Use Rotational Drag=false
   )
   &SolveForcesAndVelocity(Clamp Velocity=false, Limit Acceleration=false)
   &ScaleSpriteSizeBySpeed(
    Max Scale Factor=<1 16>
    Min Scale Factor=<1 1.5>
    Sample Scale Factor By Curve=true
    Scale Factor Curve=DataInterface() {propertyValues {Curve {keys [
        {interpMode=RCIM_Cubic, leaveTangentWeight=0.6864, tangentMode=RCTM_User, time=0, value=0.01}
        {arriveTangent=1.7192, arriveTangentWeight=0.5359, interpMode=RCIM_Cubic, leaveTangent=1.7192, tangentMode=RCTM_User, time=1, value=1}
       ]}}}
   )
   &ScaleColor(
    Scale Alpha=&FloatFromCurve(FloatCurve=DataInterface() {propertyValues {Curve {keys [<at 0 1>,<at 1 0>]}}})
    Scale RGB=&MakeVectorFromLinearColorRGB(Color=&MultiplyLinearColor(
      Linear Color=&self:SelectColorCurve(
       CurveSelect:NiagaraFloat=$User.Ranked_Curve
       ColorSelect:NiagaraFloat=&RandomRangeFloat(Minimum=0, Maximum=1)
       UnrankedColorCurve:NiagaraDataInterfaceColorCurve=DataInterface() {propertyValues {
         AlphaCurve {keys [<at 0 1>,<at 1 1>]}
         BlueCurve {keys [
           {interpMode=RCIM_Cubic, leaveTangentWeight=0.4171, tangentMode=RCTM_User, time=0, value=0.2831}
           {arriveTangent=1.9677, interpMode=RCIM_Cubic, leaveTangent=1.9677, tangentMode=RCTM_User, time=1, value=1}
          ]}
         GreenCurve {keys [
           {interpMode=RCIM_Cubic, leaveTangentWeight=0.4171, tangentMode=RCTM_User, time=0, value=0.2831}
           {arriveTangent=1.9677, interpMode=RCIM_Cubic, leaveTangent=1.9677, tangentMode=RCTM_User, time=1, value=1}
          ]}
         RedCurve {keys [
           {interpMode=RCIM_Cubic, leaveTangentWeight=0.4171, tangentMode=RCTM_User, time=0, value=0.2831}
           {arriveTangent=1.9677, interpMode=RCIM_Cubic, leaveTangent=1.9677, tangentMode=RCTM_User, time=1, value=1}
          ]}
        }}
       BronzeColorCurve:NiagaraDataInterfaceColorCurve=DataInterface() {propertyValues {
         AlphaCurve {keys [<at 0 1>,<at 1 1>]}
         BlueCurve {keys [<at 0 0.0185>,<at 1 0.1946>]}
         GreenCurve {keys [<at 0 0.0714>,<at 1 0.3557>]}
         RedCurve {keys [<at 0 0.2831>,<at 1 1>]}
        }}
       SilverColorCurve:NiagaraDataInterfaceColorCurve=DataInterface() {propertyValues {
         AlphaCurve {keys [<at 0 1>,<at 1 1>]}
         BlueCurve {keys [<at 0 0.0409>,<at 1 0.9734>]}
         GreenCurve {keys [<at 0 0.0273>,<at 1 0.9301>]}
         RedCurve {keys [<at 0 0.0194>,<at 1 0.8796>]}
        }}
       GoldColorCurve:NiagaraDataInterfaceColorCurve=DataInterface() {propertyValues {
         AlphaCurve {keys [<at 0 1>,<at 1 1>]}
         BlueCurve {keys [<at 0 0>,<at 1 0.2961>]}
         GreenCurve {keys [<at 0 0.1694>,<at 1 0.7654>]}
         RedCurve {keys [<at 0 0.2542>,<at 1 1>]}
        }}
       PlatColorCurve:NiagaraDataInterfaceColorCurve=DataInterface() {propertyValues {
         AlphaCurve {keys [<at 0 1>,<at 1 1>]}
         BlueCurve {keys [<at 0 0.0221>,<at 1 0.404>]}
         GreenCurve {keys [<at 0 0.05>,<at 1 0.9216>]}
         RedCurve {keys [<at 0 0.0021>,<at 1 0.0343>]}
        }}
       DiamondColorCurve:NiagaraDataInterfaceColorCurve=DataInterface() {propertyValues {
         AlphaCurve {keys [<at 0 1>,<at 1 1>]}
         BlueCurve {keys [<at 0 0.4>,<at 1 1>]}
         GreenCurve {keys [<at 0 0.2356>,<at 1 0.5833>]}
         RedCurve {keys [<at 0 0.0054>,<at 1 0>]}
        }}
       EliteColorCurve:NiagaraDataInterfaceColorCurve=DataInterface() {propertyValues {
         AlphaCurve {keys [<at 0 1>,<at 1 1>]}
         BlueCurve {keys [<at 0 0.0005 RCIM_Cubic>, {arriveTangent=0.3468, interpMode=RCIM_Cubic, leaveTangent=0.3468, tangentMode=RCTM_User, time=1, value=0.1}]}
         GreenCurve {keys [<at 0 0.0005 RCIM_Cubic>, {arriveTangent=0.3468, interpMode=RCIM_Cubic, leaveTangent=0.3468, tangentMode=RCTM_User, time=1, value=0.1}]}
         RedCurve {keys [<at 0 0.0005 RCIM_Cubic>, {arriveTangent=0.3468, interpMode=RCIM_Cubic, leaveTangent=0.3468, tangentMode=RCTM_User, time=1, value=0.1}]}
        }}
       ChampionColorCurve:NiagaraDataInterfaceColorCurve=DataInterface() {propertyValues {
         AlphaCurve {keys [<at 0 1>,<at 1 1>]}
         BlueCurve {keys [<at 0 0>,<at 0.5 0>,<at 1 0.0012>]}
         GreenCurve {keys [<at 0 0 RCIM_Cubic>, {arriveTangent=0.6654, interpMode=RCIM_Cubic, leaveTangent=0.6654, time=0.5, value=0.0908}, <at 1 0.6654 RCIM_Cubic>]}
         RedCurve {keys [
           <at 0 0.15 RCIM_Cubic>
           {arriveTangentWeight=0.2187, interpMode=RCIM_Cubic, leaveTangentWeight=0.2187, tangentMode=RCTM_User, time=0.5, value=1}
           <at 1 1 RCIM_Cubic>
          ]}
        }}
       UnrealColorCurve:NiagaraDataInterfaceColorCurve=DataInterface() {propertyValues {
         AlphaCurve {keys [<at 0 1>,<at 1 1>]}
         BlueCurve {keys [
           <at 0 0.2307 RCIM_Cubic>
           {arriveTangent=0.5013, interpMode=RCIM_Cubic, leaveTangent=0.5013, time=0.5, value=0.2664}
           <at 1 0.732 RCIM_Cubic>
          ]}
         GreenCurve {keys [
           <at 0 0.0252 RCIM_Cubic>
           {arriveTangent=0.6649, interpMode=RCIM_Cubic, leaveTangent=0.6649, time=0.5, value=0.0782}
           <at 1 0.6901 RCIM_Cubic>
          ]}
         RedCurve {keys [
           <at 0 0.0222 RCIM_Cubic>
           {
            arriveTangent=0.4814
            arriveTangentWeight=0.2187
            interpMode=RCIM_Cubic
            leaveTangent=0.4814
            leaveTangentWeight=0.2187
            time=0.5
            value=0.3564
           }
           <at 1 0.5036 RCIM_Cubic>
          ]}
        }}
      )
      Scale Factor=&MakeVector4_FromFloat(Float=&RandomRangeFloat(Minimum=1, Maximum=2))
     ))
    ScaleA=true
    ScaleRGB=true
   )
  ]
  rendererClasses [&NiagaraSpriteRendererProperties]
  renderers [{rendererClass=&NiagaraSpriteRendererProperties}]
 }
 Fire_Strokes {
  emitterUpdateScript [
   &EmitterState()
   &SpawnBurst_Instantaneous(
    Spawn Count=&MakeCustomIntFromBool(Bool=$User.Unranked, True INT=&RandomRangeInt(Minimum=8, Maximum=16), False INT=0)
    Spawn Time=2.06
    Use Spawn Probability=false
    UseLoopCountLimit=false
   )
  ]
  particleSpawnScript [
   &InitializeParticle(
    Color=&self:SelectColorCurve(
     CurveSelect:NiagaraFloat=$User.Ranked_Curve
     ColorSelect:NiagaraFloat=&RandomRangeFloat(Minimum=0, Maximum=1)
     UnrankedColorCurve:NiagaraDataInterfaceColorCurve=DataInterface() {propertyValues {
       AlphaCurve {keys [<at 0 1>,<at 1 1>]}
       BlueCurve {keys [
         {interpMode=RCIM_Cubic, leaveTangentWeight=0.4171, tangentMode=RCTM_User, time=0, value=0.2831}
         {arriveTangent=1.9677, interpMode=RCIM_Cubic, leaveTangent=1.9677, tangentMode=RCTM_User, time=1, value=1}
        ]}
       GreenCurve {keys [
         {interpMode=RCIM_Cubic, leaveTangentWeight=0.4171, tangentMode=RCTM_User, time=0, value=0.2831}
         {arriveTangent=1.9677, interpMode=RCIM_Cubic, leaveTangent=1.9677, tangentMode=RCTM_User, time=1, value=1}
        ]}
       RedCurve {keys [
         {interpMode=RCIM_Cubic, leaveTangentWeight=0.4171, tangentMode=RCTM_User, time=0, value=0.2831}
         {arriveTangent=1.9677, interpMode=RCIM_Cubic, leaveTangent=1.9677, tangentMode=RCTM_User, time=1, value=1}
        ]}
      }}
     BronzeColorCurve:NiagaraDataInterfaceColorCurve=DataInterface() {propertyValues {
       AlphaCurve {keys [<at 0 1>,<at 1 1>]}
       BlueCurve {keys [<at 0 0.0185>,<at 1 0.1946>]}
       GreenCurve {keys [<at 0 0.0714>,<at 1 0.3557>]}
       RedCurve {keys [<at 0 0.2831>,<at 1 1>]}
      }}
     SilverColorCurve:NiagaraDataInterfaceColorCurve=DataInterface() {propertyValues {
       AlphaCurve {keys [<at 0 1>,<at 1 1>]}
       BlueCurve {keys [<at 0 0.0409>,<at 1 0.9734>]}
       GreenCurve {keys [<at 0 0.0273>,<at 1 0.9301>]}
       RedCurve {keys [<at 0 0.0194>,<at 1 0.8796>]}
      }}
     GoldColorCurve:NiagaraDataInterfaceColorCurve=DataInterface() {propertyValues {
       AlphaCurve {keys [<at 0 1>,<at 1 1>]}
       BlueCurve {keys [<at 0 0>,<at 1 0.2961>]}
       GreenCurve {keys [<at 0 0.1694>,<at 1 0.7654>]}
       RedCurve {keys [<at 0 0.2542>,<at 1 1>]}
      }}
     PlatColorCurve:NiagaraDataInterfaceColorCurve=DataInterface() {propertyValues {
       AlphaCurve {keys [<at 0 1>,<at 1 1>]}
       BlueCurve {keys [<at 0 0.0221>,<at 1 0.404>]}
       GreenCurve {keys [<at 0 0.05>,<at 1 0.9216>]}
       RedCurve {keys [<at 0 0.0021>,<at 1 0.0343>]}
      }}
     DiamondColorCurve:NiagaraDataInterfaceColorCurve=DataInterface() {propertyValues {
       AlphaCurve {keys [<at 0 1>,<at 1 1>]}
       BlueCurve {keys [<at 0 0.4>,<at 1 1>]}
       GreenCurve {keys [<at 0 0.2356>,<at 1 0.5833>]}
       RedCurve {keys [<at 0 0.0054>,<at 1 0>]}
      }}
     EliteColorCurve:NiagaraDataInterfaceColorCurve=DataInterface() {propertyValues {
       AlphaCurve {keys [<at 0 1>,<at 1 1>]}
       BlueCurve {keys [<at 0 0.0005 RCIM_Cubic>, {arriveTangent=0.3468, interpMode=RCIM_Cubic, leaveTangent=0.3468, tangentMode=RCTM_User, time=1, value=0.1}]}
       GreenCurve {keys [<at 0 0.0005 RCIM_Cubic>, {arriveTangent=0.3468, interpMode=RCIM_Cubic, leaveTangent=0.3468, tangentMode=RCTM_User, time=1, value=0.1}]}
       RedCurve {keys [<at 0 0.0005 RCIM_Cubic>, {arriveTangent=0.3468, interpMode=RCIM_Cubic, leaveTangent=0.3468, tangentMode=RCTM_User, time=1, value=0.1}]}
      }}
     ChampionColorCurve:NiagaraDataInterfaceColorCurve=DataInterface() {propertyValues {
       AlphaCurve {keys [<at 0 1>,<at 1 1>]}
       BlueCurve {keys [<at 0 0>,<at 0.5 0>,<at 1 0.0012>]}
       GreenCurve {keys [<at 0 0 RCIM_Cubic>, {arriveTangent=0.6654, interpMode=RCIM_Cubic, leaveTangent=0.6654, time=0.5, value=0.0908}, <at 1 0.6654 RCIM_Cubic>]}
       RedCurve {keys [
         <at 0 0.15 RCIM_Cubic>
         {arriveTangentWeight=0.2187, interpMode=RCIM_Cubic, leaveTangentWeight=0.2187, tangentMode=RCTM_User, time=0.5, value=1}
         <at 1 1 RCIM_Cubic>
        ]}
      }}
     UnrealColorCurve:NiagaraDataInterfaceColorCurve=DataInterface() {propertyValues {
       AlphaCurve {keys [<at 0 1>,<at 1 1>]}
       BlueCurve {keys [<at 0 0.2307 RCIM_Cubic>, {interpMode=RCIM_Cubic, leaveTangent=0.5013, time=1, value=0.2664}]}
       GreenCurve {keys [<at 0 0.0252 RCIM_Cubic>, {interpMode=RCIM_Cubic, leaveTangent=0.6649, time=1, value=0.0782}]}
       RedCurve {keys [
         <at 0 0.0222 RCIM_Cubic>
         {arriveTangentWeight=0.2187, interpMode=RCIM_Cubic, leaveTangent=0.4814, leaveTangentWeight=0.2187, time=1, value=0.3564}
        ]}
      }}
    )
    Lifetime Max=0.333
    Lifetime Min=0.175
    Lifetime Mode=::NewEnumerator1
    Sprite Size Max=<65 150>
    Sprite Size Min=<45 75>
    Sprite Size Mode=::NewEnumerator2
    UsePositionOffset=false
    Write Material Random=true
   )
   &ShapeLocation(MaskSpawnGroup=false, Override Seed=false, Sphere Radius=2, UseOwnerScale=true)
   &AddVelocity(
    Module.Use Falloff Distance=false
    Override Seed=false
    Velocity Mode=::NewEnumerator1
    Velocity Speed=&RandomRangeFloat(Minimum=15, Maximum=40)
   )
  ]
  particleUpdateScript [
   &ParticleState()
   &ScaleSpriteSize(Uniform Curve Sprite Scale=DataInterface() {propertyValues {Curve {keys [
        <at 0 0.1035 RCIM_Cubic>
        {arriveTangent=-0.005, interpMode=RCIM_Cubic, leaveTangent=-0.005, time=0.5835, value=0.4001}
        {arriveTangent=-0.6121, interpMode=RCIM_Cubic, leaveTangent=-0.6121, tangentMode=RCTM_User, time=1, value=0.0985}
       ]}}})
   &DynamicMaterialParameters(
    Index 0 Param 1=&FloatFromCurve(FloatCurve=DataInterface() {propertyValues {Curve {keys [<at 0 0.0108>,<at 1 0.5682>]}}})
    Index 0 Param 2=&RandomRangeFloat(Minimum=0, Maximum=10)
    Module.Param0WriteEnabled=true
    Module.Param1WriteEnabled=true
    Module.Param2WriteEnabled=true
    Module.Param3WriteEnabled=true
   )
   &Drag(Use Linear Drag=true, Use Rotational Drag=true)
   &SolveForcesAndVelocity(Clamp Velocity=false, Limit Acceleration=false)
   &ScaleColor(
    Scale Alpha=&FloatFromCurve(FloatCurve=DataInterface() {propertyValues {Curve {keys [
         {arriveTangent=-9.9723, interpMode=RCIM_Cubic, leaveTangent=-9.9723, tangentMode=RCTM_User, time=0, value=1}
         <at 0.2022 0 RCIM_Cubic>
         {arriveTangent=2.1082, interpMode=RCIM_Cubic, leaveTangent=2.1082, tangentMode=RCTM_User, time=1, value=1}
        ]}}})
    Scale RGB=&MakeVectorFromLinearColorRGB(Color=&ColorFromCurve(ColorCurve=DataInterface() {propertyValues {
        AlphaCurve {keys []}
        BlueCurve {keys [<at 0.199 8.986 RCIM_Cubic>,<at 0.4 10.9338>,<at 1 2.471 RCIM_Cubic>]}
        GreenCurve {keys [<at 0.199 8.986 RCIM_Cubic>,<at 0.4 10.9338>,<at 1 2.471 RCIM_Cubic>]}
        RedCurve {keys [<at 0.199 8.986 RCIM_Cubic>,<at 0.4 10.9338>,<at 1 2.471 RCIM_Cubic>]}
       }}))
    ScaleA=true
    ScaleRGB=true
   )
  ]
  rendererClasses [&NiagaraSpriteRendererProperties]
  renderers [{rendererClass=&NiagaraSpriteRendererProperties}]
 }
 Flame_Decals {
  emitterUpdateScript [
   &EmitterState()
   &SpawnBurst_Instantaneous(Use Spawn Probability=false, UseLoopCountLimit=false)
  ]
  particleSpawnScript [
   &InitializeParticle(
    Color=&self:SelectColorCurve(
     CurveSelect:NiagaraFloat=$User.Ranked_Curve
     ColorSelect:NiagaraFloat=0.5
     UnrankedColorCurve:NiagaraDataInterfaceColorCurve=DataInterface() {propertyValues {
       AlphaCurve {keys [<at 0 1>,<at 1 1>]}
       BlueCurve {keys [
         {interpMode=RCIM_Cubic, leaveTangentWeight=0.4171, tangentMode=RCTM_User, time=0, value=0.2831}
         {arriveTangent=1.9677, interpMode=RCIM_Cubic, leaveTangent=1.9677, tangentMode=RCTM_User, time=1, value=1}
        ]}
       GreenCurve {keys [
         {interpMode=RCIM_Cubic, leaveTangentWeight=0.4171, tangentMode=RCTM_User, time=0, value=0.2831}
         {arriveTangent=1.9677, interpMode=RCIM_Cubic, leaveTangent=1.9677, tangentMode=RCTM_User, time=1, value=1}
        ]}
       RedCurve {keys [
         {interpMode=RCIM_Cubic, leaveTangentWeight=0.4171, tangentMode=RCTM_User, time=0, value=0.2831}
         {arriveTangent=1.9677, interpMode=RCIM_Cubic, leaveTangent=1.9677, tangentMode=RCTM_User, time=1, value=1}
        ]}
      }}
     BronzeColorCurve:NiagaraDataInterfaceColorCurve=DataInterface() {propertyValues {
       AlphaCurve {keys [<at 0 1>,<at 1 1>]}
       BlueCurve {keys [<at 0 0.0185>,<at 1 0.1946>]}
       GreenCurve {keys [<at 0 0.0714>,<at 1 0.3557>]}
       RedCurve {keys [<at 0 0.2831>,<at 1 1>]}
      }}
     SilverColorCurve:NiagaraDataInterfaceColorCurve=DataInterface() {propertyValues {
       AlphaCurve {keys [<at 0 1>,<at 1 1>]}
       BlueCurve {keys [<at 0 0.0409>,<at 1 0.9734>]}
       GreenCurve {keys [<at 0 0.0273>,<at 1 0.9301>]}
       RedCurve {keys [<at 0 0.0194>,<at 1 0.8796>]}
      }}
     GoldColorCurve:NiagaraDataInterfaceColorCurve=DataInterface() {propertyValues {
       AlphaCurve {keys [<at 0 1>,<at 1 1>]}
       BlueCurve {keys [<at 0 0>,<at 1 0.2961>]}
       GreenCurve {keys [<at 0 0.1694>,<at 1 0.7654>]}
       RedCurve {keys [<at 0 0.2542>,<at 1 1>]}
      }}
     PlatColorCurve:NiagaraDataInterfaceColorCurve=DataInterface() {propertyValues {
       AlphaCurve {keys [<at 0 1>,<at 1 1>]}
       BlueCurve {keys [<at 0 0.0221>,<at 1 0.404>]}
       GreenCurve {keys [<at 0 0.05>,<at 1 0.9216>]}
       RedCurve {keys [<at 0 0.0021>,<at 1 0.0343>]}
      }}
     DiamondColorCurve:NiagaraDataInterfaceColorCurve=DataInterface() {propertyValues {
       AlphaCurve {keys [<at 0 1>,<at 1 1>]}
       BlueCurve {keys [<at 0 0.4>,<at 1 1>]}
       GreenCurve {keys [<at 0 0.2356>,<at 1 0.5833>]}
       RedCurve {keys [<at 0 0.0054>,<at 1 0>]}
      }}
     EliteColorCurve:NiagaraDataInterfaceColorCurve=DataInterface() {propertyValues {
       AlphaCurve {keys [<at 0 1>,<at 1 1>]}
       BlueCurve {keys [<at 0 0.0005 RCIM_Cubic>, {arriveTangent=0.3468, interpMode=RCIM_Cubic, leaveTangent=0.3468, tangentMode=RCTM_User, time=1, value=0.1}]}
       GreenCurve {keys [<at 0 0.0005 RCIM_Cubic>, {arriveTangent=0.3468, interpMode=RCIM_Cubic, leaveTangent=0.3468, tangentMode=RCTM_User, time=1, value=0.1}]}
       RedCurve {keys [<at 0 0.0005 RCIM_Cubic>, {arriveTangent=0.3468, interpMode=RCIM_Cubic, leaveTangent=0.3468, tangentMode=RCTM_User, time=1, value=0.1}]}
      }}
     ChampionColorCurve:NiagaraDataInterfaceColorCurve=DataInterface() {propertyValues {
       AlphaCurve {keys [<at 0 1>,<at 1 1>]}
       BlueCurve {keys [<at 0 0>,<at 0.5 0>,<at 1 0.0012>]}
       GreenCurve {keys [<at 0 0 RCIM_Cubic>, {arriveTangent=0.6654, interpMode=RCIM_Cubic, leaveTangent=0.6654, time=0.5, value=0.0908}, <at 1 0.6654 RCIM_Cubic>]}
       RedCurve {keys [
         <at 0 0.15 RCIM_Cubic>
         {arriveTangentWeight=0.2187, interpMode=RCIM_Cubic, leaveTangentWeight=0.2187, tangentMode=RCTM_User, time=0.5, value=1}
         <at 1 1 RCIM_Cubic>
        ]}
      }}
     UnrealColorCurve:NiagaraDataInterfaceColorCurve=DataInterface() {propertyValues {
       AlphaCurve {keys [<at 0 1>,<at 1 1>]}
       BlueCurve {keys [<at 0 0.2307 RCIM_Cubic>, {interpMode=RCIM_Cubic, leaveTangent=0.5013, time=1, value=0.2664}]}
       GreenCurve {keys [<at 0 0.0252 RCIM_Cubic>, {interpMode=RCIM_Cubic, leaveTangent=0.6649, time=1, value=0.0782}]}
       RedCurve {keys [
         <at 0 0.0222 RCIM_Cubic>
         {arriveTangentWeight=0.2187, interpMode=RCIM_Cubic, leaveTangent=0.4814, leaveTangentWeight=0.2187, time=1, value=0.3564}
        ]}
      }}
    )
    Lifetime=4
    Sprite Size Mode=::NewEnumerator3
    Uniform Sprite Size=50
    UsePositionOffset=false
    Write Material Random=true
   )
   &CameraOffset(Camera Offset Amount=20)
  ]
  particleUpdateScript [
   &ParticleState()
   &DynamicMaterialParameters(
    Index 0 Param 1=2
    Index 0 Param 2=&FloatFromCurve(FloatCurve=DataInterface() {propertyValues {Curve {keys [<at 0 2>,<at 1 4>]}}}, CurveIndex=$Emitter.NormalizedLoopAge)
    Index 0 Param 3=2
    Index 1 Param 2=-0.25
    Index 1 Param 3=0.01
    Index 1 Param 4=-0.02
    Module.Param0WriteEnabled=true
    Module.Param0WriteEnabled1=true
    Module.Param1WriteEnabled=true
    Module.Param1WriteEnabled1=true
    Module.Param2WriteEnabled=true
    Module.Param2WriteEnabled1=true
    Module.Param3WriteEnabled=true
    Module.Param3WriteEnabled1=true
    Write Parameter Index 1=true
   )
   &ScaleColor(
    Scale Alpha=&FloatFromCurve(
     FloatCurve=DataInterface() {propertyValues {Curve {keys [{interpMode=RCIM_Cubic, leaveTangentWeight=0.6587, tangentMode=RCTM_User, time=3.257, value=1}, <at 4 0 RCIM_Cubic>]}}}
     CurveIndex=$Emitter.Age
    )
    ScaleA=true
    ScaleRGB=true
   )
  ]
  rendererClasses [&NiagaraSpriteRendererProperties]
  renderers [{rendererClass=&NiagaraSpriteRendererProperties}]
 }
 Flame_Meshes {
  emitterUpdateScript [
   &EmitterState()
   &SpawnBurst_Instantaneous(
    Spawn Count=&MakeCustomIntFromBool(Bool=$User.Unranked, True INT=0, False INT=20)
    Spawn Time=2.0117
    Use Spawn Probability=false
    UseLoopCountLimit=false
   )
   &SpawnRate(
    SpawnRate=&MakeCustomFloatFromBool(
     Bool=$User.Unranked
     True Float=0
     False Float=&Multiply_Float(
      A=8
      B=&FloatFromCurve(FloatCurve=DataInterface() {propertyValues {Curve {keys [<at 3.257 1>,<at 4 0>]}}}, CurveIndex=$Emitter.Age)
     )
    )
    Use Spawn Probability=false
   )
  ]
  particleSpawnScript [
   &InitializeParticle(
    Color=<rgba 1 1 1 0.25>
    Lifetime Max=0.825
    Lifetime Min=0.6
    Lifetime Mode=::NewEnumerator1
    Mesh Scale=&Multiply_VectorByFloat(
     Vector=&RandomRangeVector(Minimum=<0.1 0.1 0.2>, Maximum=<0.2 0.2 0.5>)
     Float=$System.NormalizedFirstBurst
    )
    Mesh Scale Mode=::NewEnumerator4
    UsePositionOffset=false
    Write Material Random=true
   )
   &ShapeLocation(
    Cylinder Height=1
    Cylinder Height Midpoint=0
    Cylinder Radius=&Multiply_Float(
     A=10
     B=&FloatFromCurve(
      FloatCurve=DataInterface() {propertyValues {Curve {keys [<at 0 0.7 RCIM_Constant>,<at 0.34 1.2>,<at 0.4 1.2>]}}}
      CurveIndex=$Emitter.NormalizedLoopAge
     )
    )
    MaskSpawnGroup=false
    Override Seed=false
    Shape Primitive=::NewEnumerator1
    Surface Only=false
    UseOwnerScale=true
   )
   &CameraOffset(Camera Offset Amount=10)
  ]
  particleUpdateScript [
   &ParticleState()
   &ScaleMeshSize(Scale Factor=&VectorFromFloat(Value=&FloatFromCurve(FloatCurve=DataInterface() {propertyValues {Curve {keys [
          <at 0 0.65 RCIM_Cubic>
          {arriveTangent=-0.565, interpMode=RCIM_Cubic, leaveTangent=-0.565, tangentMode=RCTM_User, time=0.2, value=1}
          {interpMode=RCIM_Cubic, tangentMode=RCTM_User, time=1, value=0.7}
         ]}}})))
   &SubUVAnimation(
    Mesh Renderer=DataInterface() {propertyValues {MeshRenderer=&self:Flame_Meshes_1.NiagaraMeshRendererProperties_3}}
    Renderer Type:ENiagaraMeshOrSprite=::NewEnumerator1
    Start Frame Range Override=&RandomRangeInt(Minimum=0, Maximum=10)
    SubUV Lookup Index Scale:NiagaraFloat=&RandomRangeFloat(Minimum=0.9, Maximum=1.1)
    UseEndFrame=true
    UseStartFrame=true
   )
   &DynamicMaterialParameters(
    Index 0 Param 1=&RandomRangeFloat(Minimum=0, Maximum=2)
    Index 0 Param 2=&RandomRangeFloat(Minimum=0, Maximum=2)
    Index 0 Param 3=8
    Index 0 Param 4=$Particles.SubImageIndex
    Index 1 Param 1=&RandomRangeFloat(Minimum=-1, Maximum=1)
    Index 1 Param 2=&RandomRangeFloat(Minimum=0, Maximum=3)
    Index 1 Param 3=&FloatFromCurve(FloatCurve=DataInterface() {propertyValues {Curve {keys [
         {arriveTangent=4.1785, interpMode=RCIM_Cubic, leaveTangent=4.1785, tangentMode=RCTM_User, time=0, value=0}
         {arriveTangent=0.9995, interpMode=RCIM_Cubic, leaveTangent=0.9995, tangentMode=RCTM_User, time=0.2223, value=0.6825}
         {arriveTangent=-0.0405, interpMode=RCIM_Cubic, leaveTangent=-0.0405, tangentMode=RCTM_User, time=0.7, value=0.9}
         <at 1 0.25>
        ]}}})
    Module.Param0WriteEnabled=true
    Module.Param0WriteEnabled1=true
    Module.Param1WriteEnabled=true
    Module.Param1WriteEnabled1=true
    Module.Param2WriteEnabled=true
    Module.Param2WriteEnabled1=true
    Module.Param3WriteEnabled=true
    Module.Param3WriteEnabled1=false
    Write Parameter Index 1=true
   )
   &ScaleColor(
    Scale Alpha=&FloatFromCurve(
     FloatCurve=DataInterface() {propertyValues {Curve {keys [
         {arriveTangent=9.0725, interpMode=RCIM_Cubic, leaveTangent=9.0725, tangentMode=RCTM_User, time=0, value=0}
         {arriveTangent=0.1459, interpMode=RCIM_Cubic, leaveTangent=0.1459, tangentMode=RCTM_User, time=0.1408, value=0.9847}
         {interpMode=RCIM_Cubic, tangentMode=RCTM_User, time=0.4, value=1}
         {arriveTangent=-0.0405, interpMode=RCIM_Cubic, leaveTangent=-0.0405, tangentMode=RCTM_User, time=1, value=0.1}
        ]}}}
     Scale Curve=&FloatFromCurve(
      FloatCurve=DataInterface() {propertyValues {Curve {keys [{arriveTangent=-2.4644, interpMode=RCIM_Cubic, leaveTangent=-2.4644, tangentMode=RCTM_User, time=3.257, value=1}, <at 4 0 RCIM_Cubic>]}}}
      CurveIndex=$Emitter.Age
     )
    )
    Scale RGB=&VectorFromFloat(Value=&FloatFromCurve(FloatCurve=DataInterface() {propertyValues {Curve {keys [
          {arriveTangent=9.0725, interpMode=RCIM_Cubic, leaveTangent=9.0725, tangentMode=RCTM_User, time=0, value=0}
          {arriveTangent=-2.5249, interpMode=RCIM_Cubic, leaveTangent=-2.5249, tangentMode=RCTM_User, time=0.1408, value=0.9847}
          {arriveTangent=-0.0405, interpMode=RCIM_Cubic, leaveTangent=-0.0405, tangentMode=RCTM_User, time=1, value=0.1}
         ]}}}))
    ScaleA=true
    ScaleRGB=true
   )
   &SolveForcesAndVelocity(Clamp Velocity=false, Limit Acceleration=false)
  ]
  rendererClasses [&NiagaraMeshRendererProperties]
  renderers [
   {rendererClass=&NiagaraMeshRendererProperties, rendererIndex=0}
   {rendererClass=&NiagaraMeshRendererProperties, rendererIndex=1}
  ]
 }
 Flames_BKGD {
  emitterUpdateScript [
   &EmitterState()
   &SpawnBurst_Instantaneous(
    Spawn Count=&MakeCustomIntFromBool(Bool=$User.Unranked, True INT=0, False INT=2)
    Spawn Group=1
    Spawn Time=2.027
    Use Spawn Probability=false
    UseLoopCountLimit=false
   )
   &SpawnRate(
    SpawnGroup=1
    SpawnRate=&MakeCustomFloatFromBool(
     Bool=$User.Unranked
     True Float=0
     False Float=&Multiply_Float(
      A=2
      B=&FloatFromCurve(FloatCurve=DataInterface() {propertyValues {Curve {keys [<at 3.257 1>,<at 4 0>]}}}, CurveIndex=$Emitter.Age, Scale Curve=5)
     )
    )
    Use Spawn Probability=false
   )
  ]
  particleSpawnScript [
   &InitializeParticle(
    Color=&self:SelectColorCurve(
     CurveSelect:NiagaraFloat=$User.Ranked_Curve
     ColorSelect:NiagaraFloat=0.5
     UnrankedColorCurve:NiagaraDataInterfaceColorCurve=DataInterface() {propertyValues {
       AlphaCurve {keys [<at 0 1>,<at 1 1>]}
       BlueCurve {keys [
         {interpMode=RCIM_Cubic, leaveTangentWeight=0.4171, tangentMode=RCTM_User, time=0, value=0.2831}
         {arriveTangent=1.9677, interpMode=RCIM_Cubic, leaveTangent=1.9677, tangentMode=RCTM_User, time=1, value=1}
        ]}
       GreenCurve {keys [
         {interpMode=RCIM_Cubic, leaveTangentWeight=0.4171, tangentMode=RCTM_User, time=0, value=0.2831}
         {arriveTangent=1.9677, interpMode=RCIM_Cubic, leaveTangent=1.9677, tangentMode=RCTM_User, time=1, value=1}
        ]}
       RedCurve {keys [
         {interpMode=RCIM_Cubic, leaveTangentWeight=0.4171, tangentMode=RCTM_User, time=0, value=0.2831}
         {arriveTangent=1.9677, interpMode=RCIM_Cubic, leaveTangent=1.9677, tangentMode=RCTM_User, time=1, value=1}
        ]}
      }}
     BronzeColorCurve:NiagaraDataInterfaceColorCurve=DataInterface() {propertyValues {
       AlphaCurve {keys [<at 0 1>,<at 1 1>]}
       BlueCurve {keys [<at 0 0.0185>,<at 1 0.1946>]}
       GreenCurve {keys [<at 0 0.0714>,<at 1 0.3557>]}
       RedCurve {keys [<at 0 0.2831>,<at 1 1>]}
      }}
     SilverColorCurve:NiagaraDataInterfaceColorCurve=DataInterface() {propertyValues {
       AlphaCurve {keys [<at 0 1>,<at 1 1>]}
       BlueCurve {keys [<at 0 0.0409>,<at 1 0.9734>]}
       GreenCurve {keys [<at 0 0.0273>,<at 1 0.9301>]}
       RedCurve {keys [<at 0 0.0194>,<at 1 0.8796>]}
      }}
     GoldColorCurve:NiagaraDataInterfaceColorCurve=DataInterface() {propertyValues {
       AlphaCurve {keys [<at 0 1>,<at 1 1>]}
       BlueCurve {keys [<at 0 0>,<at 1 0.2961>]}
       GreenCurve {keys [<at 0 0.1694>,<at 1 0.7654>]}
       RedCurve {keys [<at 0 0.2542>,<at 1 1>]}
      }}
     PlatColorCurve:NiagaraDataInterfaceColorCurve=DataInterface() {propertyValues {
       AlphaCurve {keys [<at 0 1>,<at 1 1>]}
       BlueCurve {keys [<at 0 0.0221>,<at 1 0.404>]}
       GreenCurve {keys [<at 0 0.05>,<at 1 0.9216>]}
       RedCurve {keys [<at 0 0.0021>,<at 1 0.0343>]}
      }}
     DiamondColorCurve:NiagaraDataInterfaceColorCurve=DataInterface() {propertyValues {
       AlphaCurve {keys [<at 0 1>,<at 1 1>]}
       BlueCurve {keys [<at 0 0.4>,<at 1 1>]}
       GreenCurve {keys [<at 0 0.2356>,<at 1 0.5833>]}
       RedCurve {keys [<at 0 0.0054>,<at 1 0>]}
      }}
     EliteColorCurve:NiagaraDataInterfaceColorCurve=DataInterface() {propertyValues {
       AlphaCurve {keys [<at 0 1>,<at 1 1>]}
       BlueCurve {keys [<at 0 0.0005 RCIM_Cubic>, {arriveTangent=0.3468, interpMode=RCIM_Cubic, leaveTangent=0.3468, tangentMode=RCTM_User, time=1, value=0.1}]}
       GreenCurve {keys [<at 0 0.0005 RCIM_Cubic>, {arriveTangent=0.3468, interpMode=RCIM_Cubic, leaveTangent=0.3468, tangentMode=RCTM_User, time=1, value=0.1}]}
       RedCurve {keys [<at 0 0.0005 RCIM_Cubic>, {arriveTangent=0.3468, interpMode=RCIM_Cubic, leaveTangent=0.3468, tangentMode=RCTM_User, time=1, value=0.1}]}
      }}
     ChampionColorCurve:NiagaraDataInterfaceColorCurve=DataInterface() {propertyValues {
       AlphaCurve {keys [<at 0 1>,<at 1 1>]}
       BlueCurve {keys [<at 0 0>,<at 0.5 0>,<at 1 0.0012>]}
       GreenCurve {keys [<at 0 0 RCIM_Cubic>, {arriveTangent=0.6654, interpMode=RCIM_Cubic, leaveTangent=0.6654, time=0.5, value=0.0908}, <at 1 0.6654 RCIM_Cubic>]}
       RedCurve {keys [
         <at 0 0.15 RCIM_Cubic>
         {arriveTangentWeight=0.2187, interpMode=RCIM_Cubic, leaveTangentWeight=0.2187, tangentMode=RCTM_User, time=0.5, value=1}
         <at 1 1 RCIM_Cubic>
        ]}
      }}
     UnrealColorCurve:NiagaraDataInterfaceColorCurve=DataInterface() {propertyValues {
       AlphaCurve {keys [<at 0 1>,<at 1 1>]}
       BlueCurve {keys [<at 0 0.2307 RCIM_Cubic>, {interpMode=RCIM_Cubic, leaveTangent=0.5013, time=1, value=0.2664}]}
       GreenCurve {keys [<at 0 0.0252 RCIM_Cubic>, {interpMode=RCIM_Cubic, leaveTangent=0.6649, time=1, value=0.0782}]}
       RedCurve {keys [
         <at 0 0.0222 RCIM_Cubic>
         {arriveTangentWeight=0.2187, interpMode=RCIM_Cubic, leaveTangent=0.4814, leaveTangentWeight=0.2187, time=1, value=0.3564}
        ]}
      }}
    )
    Lifetime=&MaskFloatBySpawnGroup(Masked Float=1.2, Passthrough Float=0.6)
    Position Mode=::NewEnumerator0
    Sprite Size Mode=::NewEnumerator3
    Uniform Sprite Size=120
    Write Material Random=true
   )
   &CameraOffset(Camera Offset Amount=-10)
   &AddVelocity(Rotation Coordinate Space=::World, Velocity=<0 0 25>)
  ]
  particleUpdateScript [
   &ParticleState()
   &ScaleSpriteSize(Uniform Curve Sprite Scale=DataInterface() {propertyValues {Curve {keys [
        <at 0 0.4846 RCIM_Cubic>
        {arriveTangent=0.0074, interpMode=RCIM_Cubic, leaveTangent=0.0074, time=0.7003, value=1.0028}
        <at 1 0.492 RCIM_Cubic>
       ]}}})
   &ScaleColor(
    Scale Alpha=&FloatFromCurve(
     FloatCurve=DataInterface() {propertyValues {Curve {keys [
         {arriveTangent=5.7529, interpMode=RCIM_Cubic, leaveTangent=5.7529, tangentMode=RCTM_User, time=0, value=0}
         {arriveTangent=0.4482, interpMode=RCIM_Cubic, leaveTangent=0.4482, tangentMode=RCTM_User, time=0.1221, value=0.055}
         {arriveTangent=-0.1774, interpMode=RCIM_Cubic, leaveTangent=-0.1774, tangentMode=RCTM_User, time=0.8633, value=0.0625}
         {arriveTangent=-2.985, interpMode=RCIM_Cubic, leaveTangent=-2.985, tangentMode=RCTM_User, time=1, value=0}
        ]}}}
     Scale Curve=&FloatFromCurve(FloatCurve=DataInterface() {propertyValues {Curve {keys [<at 3.257 1>,<at 4 0>]}}}, CurveIndex=$Emitter.Age)
    )
    ScaleA=true
    ScaleRGB=true
   )
   &SolveForcesAndVelocity(Clamp Velocity=false, Limit Acceleration=false)
   &DynamicMaterialParameters(
    Index 0 Param 1=&FloatFromCurve(FloatCurve=DataInterface() {propertyValues {Curve {keys [
         {interpMode=RCIM_Cubic, leaveTangentWeight=0.3342, tangentMode=RCTM_User, time=0, value=0}
         {arriveTangent=-1.7842, arriveTangentWeight=0.3333, interpMode=RCIM_Cubic, leaveTangent=-1.7842, tangentMode=RCTM_User, time=1, value=-0.4908}
        ]}}})
    Index 0 Param 2=&RandomRangeFloat(Minimum=0, Maximum=100)
    Module.Param0WriteEnabled=true
    Module.Param1WriteEnabled=true
    Module.Param2WriteEnabled=false
    Module.Param3WriteEnabled=false
   )
  ]
  rendererClasses [&NiagaraSpriteRendererProperties]
  renderers [{rendererClass=&NiagaraSpriteRendererProperties}]
 }
 Flamin_Ribbon {
  emitterUpdateScript [
   &EmitterState()
   &SpawnBurst_Instantaneous(
    Spawn Count=&MakeCustomIntFromBool(Bool=$User.Unranked, True INT=0, False INT=5)
    Use Spawn Probability=false
    UseLoopCountLimit=false
   )
  ]
  particleSpawnScript [
   &InitializeParticle(Lifetime=0.1, Ribbon Width=30, Ribbon Width Mode=::NewEnumerator1, UsePositionOffset=false, Write Material Random=true)
   /SetVariables(Particles.RibbonLinkOrder:NiagaraFloat=&ReturnNormalizedExecIndex()) {isSetParameters=true}
   &CameraOffset(Camera Offset Amount=50)
  ]
  particleUpdateScript [
   &ParticleState(Kill Particles When Lifetime Has Elapsed=false, Let Infinitely Lived Particles Die When Emitter Deactivates=true)
   &PendulumSetup(
    Module.Enable Spring Driver=true
    Pendulum Length:NiagaraFloat=&Lerp_Float(
     A=&Multiply_Float(A=50, B=$System.NormalizedFirstBurst)
     B=0
     Alpha=&ReturnNormalizedExecIndex()
    )
    Pendulum Rest Axis:Vector3f=<0 0 1>
    Rigid Pendulum=false
    "Spring Driven Constraint (Tightness)"=1
   )
   &DynamicMaterialParameters(
    Index 0 Param 1=&Multiply_Float(A=$Emitter.Age, B=0.8)
    Index 0 Param 2=1
    Module.Param0WriteEnabled=true
    Module.Param1WriteEnabled=true
    Module.Param2WriteEnabled=true
    Module.Param3WriteEnabled=true
   )
   &WindForce(
    Frequency=5
    Offset By Time:Vector3f=<0 0 100>
    Scale=0.1
    Wind Speed=&Add_Vector(
     A=&Multiply_VectorByFloat(
      Vector=<0 0 1000>
      Float=&Lerp_Float(A=0.27, B=1, Alpha=&ReturnNormalizedExecIndex(Normalized Index Includes One=false))
     )
     B=&Multiply_VectorByFloat(Vector=<1 1 1>, Float=0.6)
    )
   )
   &Drag(Drag=6, Use Linear Drag=true, Use Rotational Drag=false)
   &SolveForcesAndVelocity(Clamp Velocity=false, Limit Acceleration=false)
   &PendulumConstraint()
   &CalculateAccurateVelocity()
   &ScaleColor(
    Scale Alpha=&FloatFromCurve(
     FloatCurve=DataInterface() {propertyValues {Curve {keys [{arriveTangent=-2.4644, interpMode=RCIM_Cubic, leaveTangent=-2.4644, tangentMode=RCTM_User, time=3.257, value=1}, <at 4 0 RCIM_Cubic>]}}}
     CurveIndex=$Emitter.Age
    )
    ScaleA=true
    ScaleRGB=true
   )
  ]
  rendererClasses [&NiagaraRibbonRendererProperties]
  renderers [{rendererClass=&NiagaraRibbonRendererProperties}]
 }
 Light {
  emitterUpdateScript [
   &EmitterState()
   &SpawnBurst_Instantaneous(Use Spawn Probability=false, UseLoopCountLimit=false)
  ]
  particleSpawnScript [&InitializeParticle(
    Color=&self:SelectColorCurve(
     CurveSelect:NiagaraFloat=$User.Ranked_Curve
     ColorSelect:NiagaraFloat=0.5
     UnrankedColorCurve:NiagaraDataInterfaceColorCurve=DataInterface() {propertyValues {
       AlphaCurve {keys [<at 0 1>,<at 1 1>]}
       BlueCurve {keys [
         {interpMode=RCIM_Cubic, leaveTangentWeight=0.4171, tangentMode=RCTM_User, time=0, value=0.2831}
         {arriveTangent=1.9677, interpMode=RCIM_Cubic, leaveTangent=1.9677, tangentMode=RCTM_User, time=1, value=1}
        ]}
       GreenCurve {keys [
         {interpMode=RCIM_Cubic, leaveTangentWeight=0.4171, tangentMode=RCTM_User, time=0, value=0.2831}
         {arriveTangent=1.9677, interpMode=RCIM_Cubic, leaveTangent=1.9677, tangentMode=RCTM_User, time=1, value=1}
        ]}
       RedCurve {keys [
         {interpMode=RCIM_Cubic, leaveTangentWeight=0.4171, tangentMode=RCTM_User, time=0, value=0.2831}
         {arriveTangent=1.9677, interpMode=RCIM_Cubic, leaveTangent=1.9677, tangentMode=RCTM_User, time=1, value=1}
        ]}
      }}
     BronzeColorCurve:NiagaraDataInterfaceColorCurve=DataInterface() {propertyValues {
       AlphaCurve {keys [<at 0 1>,<at 1 1>]}
       BlueCurve {keys [<at 0 0.0185>,<at 1 0.1946>]}
       GreenCurve {keys [<at 0 0.0714>,<at 1 0.3557>]}
       RedCurve {keys [<at 0 0.2831>,<at 1 1>]}
      }}
     SilverColorCurve:NiagaraDataInterfaceColorCurve=DataInterface() {propertyValues {
       AlphaCurve {keys [<at 0 1>,<at 1 1>]}
       BlueCurve {keys [<at 0 0.0409>,<at 1 0.9734>]}
       GreenCurve {keys [<at 0 0.0273>,<at 1 0.9301>]}
       RedCurve {keys [<at 0 0.0194>,<at 1 0.8796>]}
      }}
     GoldColorCurve:NiagaraDataInterfaceColorCurve=DataInterface() {propertyValues {
       AlphaCurve {keys [<at 0 1>,<at 1 1>]}
       BlueCurve {keys [<at 0 0>,<at 1 0.2961>]}
       GreenCurve {keys [<at 0 0.1694>,<at 1 0.7654>]}
       RedCurve {keys [<at 0 0.2542>,<at 1 1>]}
      }}
     PlatColorCurve:NiagaraDataInterfaceColorCurve=DataInterface() {propertyValues {
       AlphaCurve {keys [<at 0 1>,<at 1 1>]}
       BlueCurve {keys [<at 0 0.0221>,<at 1 0.404>]}
       GreenCurve {keys [<at 0 0.05>,<at 1 0.9216>]}
       RedCurve {keys [<at 0 0.0021>,<at 1 0.0343>]}
      }}
     DiamondColorCurve:NiagaraDataInterfaceColorCurve=DataInterface() {propertyValues {
       AlphaCurve {keys [<at 0 1>,<at 1 1>]}
       BlueCurve {keys [<at 0 0.4>,<at 1 1>]}
       GreenCurve {keys [<at 0 0.2356>,<at 1 0.5833>]}
       RedCurve {keys [<at 0 0.0054>,<at 1 0>]}
      }}
     EliteColorCurve:NiagaraDataInterfaceColorCurve=DataInterface() {propertyValues {
       AlphaCurve {keys [<at 0 1>,<at 1 1>]}
       BlueCurve {keys [<at 0 0.0005 RCIM_Cubic>, {arriveTangent=0.3468, interpMode=RCIM_Cubic, leaveTangent=0.3468, tangentMode=RCTM_User, time=1, value=0.1}]}
       GreenCurve {keys [<at 0 0.0005 RCIM_Cubic>, {arriveTangent=0.3468, interpMode=RCIM_Cubic, leaveTangent=0.3468, tangentMode=RCTM_User, time=1, value=0.1}]}
       RedCurve {keys [<at 0 0.0005 RCIM_Cubic>, {arriveTangent=0.3468, interpMode=RCIM_Cubic, leaveTangent=0.3468, tangentMode=RCTM_User, time=1, value=0.1}]}
      }}
     ChampionColorCurve:NiagaraDataInterfaceColorCurve=DataInterface() {propertyValues {
       AlphaCurve {keys [<at 0 1>,<at 1 1>]}
       BlueCurve {keys [<at 0 0>,<at 0.5 0>,<at 1 0.0012>]}
       GreenCurve {keys [<at 0 0 RCIM_Cubic>, {arriveTangent=0.6654, interpMode=RCIM_Cubic, leaveTangent=0.6654, time=0.5, value=0.0908}, <at 1 0.6654 RCIM_Cubic>]}
       RedCurve {keys [
         <at 0 0.15 RCIM_Cubic>
         {arriveTangentWeight=0.2187, interpMode=RCIM_Cubic, leaveTangentWeight=0.2187, tangentMode=RCTM_User, time=0.5, value=1}
         <at 1 1 RCIM_Cubic>
        ]}
      }}
     UnrealColorCurve:NiagaraDataInterfaceColorCurve=DataInterface() {propertyValues {
       AlphaCurve {keys [<at 0 1>,<at 1 1>]}
       BlueCurve {keys [<at 0 0.2307 RCIM_Cubic>, {interpMode=RCIM_Cubic, leaveTangent=0.5013, time=1, value=0.2664}]}
       GreenCurve {keys [<at 0 0.0252 RCIM_Cubic>, {interpMode=RCIM_Cubic, leaveTangent=0.6649, time=1, value=0.0782}]}
       RedCurve {keys [
         <at 0 0.0222 RCIM_Cubic>
         {arriveTangentWeight=0.2187, interpMode=RCIM_Cubic, leaveTangent=0.4814, leaveTangentWeight=0.2187, time=1, value=0.3564}
        ]}
      }}
    )
    Lifetime=0.5
    Position Offset=<0 0 20>
    Sprite Size Mode=::NewEnumerator3
    Uniform Sprite Size=0
    UsePositionOffset=true
    Write Material Random=true
   )]
  particleUpdateScript [
   &ParticleState(
    Kill Particles When Lifetime Has Elapsed=false
    Let Infinitely Lived Particles Die When Emitter Deactivates=true
    Loop Particles Lifetime=true
   )
   /SetVariables(
    Particles.LightIntensity:NiagaraFloat=&Multiply_Float(
     A=250
     B=&FloatFromCurve(FloatCurve=DataInterface() {propertyValues {Curve {keys [<at 3.257 1>,<at 4 0>]}}}, CurveIndex=$Emitter.Age)
    )
    Particles.LightRendererColor:LinearColor=&MultiplyLinearColor(
     Linear Color=$Particles.Initial.Color
     Scale Factor=&MakeVector4_FromFloat(Float=&Multiply_Float(
       A=50
       B=&FloatFromCurve(FloatCurve=DataInterface() {propertyValues {Curve {keys [<at 3.257 1>,<at 4 0>]}}}, CurveIndex=$Emitter.Age)
      ))
    )
   ) {isSetParameters=true}
   &InheritSourceMovement(ApplyForce=false, ApplyPosition=true, ApplyPositionFalloff=false)
   &VectorNoiseForce(Force Amount=300, Module.Use Random Noise Time Interval=true, Random Noise Time Interval=0.05)
   &Drag(Drag=5, Use Linear Drag=true, Use Rotational Drag=true)
   &SolveForcesAndVelocity(Clamp Velocity=false, Limit Acceleration=false)
  ]
  rendererClasses [
   &NiagaraSpriteRendererProperties
   &NiagaraComponentRendererProperties
   &NiagaraLightRendererProperties
  ]
  renderers [
   {rendererClass=&NiagaraSpriteRendererProperties}
   {rendererClass=&NiagaraComponentRendererProperties, rendererIndex=1}
   {rendererClass=&NiagaraLightRendererProperties, rendererIndex=2}
  ]
 }
 Unranked_BKGD {
  emitterUpdateScript [
   &EmitterState()
   &SpawnBurst_Instantaneous(
    Spawn Count=&MakeCustomIntFromBool(Bool=$User.Unranked, True INT=2, False INT=0)
    Spawn Group=1
    Spawn Time=2.01
    Use Spawn Probability=false
    UseLoopCountLimit=false
   )
   &SpawnRate(
    SpawnGroup=1
    SpawnRate=&MakeCustomFloatFromBool(
     Bool=$User.Unranked
     True Float=&Multiply_Float(
      A=2
      B=&FloatFromCurve(FloatCurve=DataInterface() {propertyValues {Curve {keys [<at 4.572 1>,<at 5.1816 0>]}}}, CurveIndex=$Emitter.Age)
     )
     False Float=0
    )
    Use Spawn Probability=false
   )
  ]
  particleSpawnScript [
   &InitializeParticle(
    Color=&ColorFromCurve(ColorCurve=DataInterface() {propertyValues {
       AlphaCurve {keys [<at 0 1>,<at 1 1>]}
       BlueCurve {keys [
         {interpMode=RCIM_Cubic, leaveTangentWeight=0.4171, tangentMode=RCTM_User, time=0, value=0.2831}
         {arriveTangent=1.9677, interpMode=RCIM_Cubic, leaveTangent=1.9677, tangentMode=RCTM_User, time=1, value=1}
        ]}
       GreenCurve {keys [
         {interpMode=RCIM_Cubic, leaveTangentWeight=0.4171, tangentMode=RCTM_User, time=0, value=0.2831}
         {arriveTangent=1.9677, interpMode=RCIM_Cubic, leaveTangent=1.9677, tangentMode=RCTM_User, time=1, value=1}
        ]}
       RedCurve {keys [
         {interpMode=RCIM_Cubic, leaveTangentWeight=0.4171, tangentMode=RCTM_User, time=0, value=0.2831}
         {arriveTangent=1.9677, interpMode=RCIM_Cubic, leaveTangent=1.9677, tangentMode=RCTM_User, time=1, value=1}
        ]}
      }})
    Lifetime=&MaskFloatBySpawnGroup(Masked Float=1.2, Passthrough Float=0.6)
    Position Offset=<-5 0 0>
    Sprite Size Mode=::NewEnumerator3
    Uniform Sprite Size=100
    UsePositionOffset=true
    Write Material Random=true
   )
   &CameraOffset()
  ]
  particleUpdateScript [
   &ParticleState()
   &ScaleSpriteSize(Uniform Curve Sprite Scale=DataInterface() {propertyValues {Curve {keys [<at 0 0.25 RCIM_Cubic>,<at 0.5 0.5 RCIM_Cubic>,<at 1 0.25 RCIM_Cubic>]}}})
   &ScaleColor(
    Scale Alpha=&FloatFromCurve(
     FloatCurve=DataInterface() {propertyValues {Curve {keys [
         {arriveTangent=2.8438, interpMode=RCIM_Cubic, leaveTangent=2.8438, tangentMode=RCTM_User, time=0, value=0}
         <at 0.5 1 RCIM_Cubic>
         {arriveTangent=-3.0532, interpMode=RCIM_Cubic, leaveTangent=-3.0532, tangentMode=RCTM_User, time=1, value=0}
        ]}}}
     Scale Curve=&FloatFromCurve(
      FloatCurve=DataInterface() {propertyValues {Curve {keys [{interpMode=RCIM_Cubic, leaveTangentWeight=0.6587, tangentMode=RCTM_User, time=3.257, value=1}, <at 4 0 RCIM_Cubic>]}}}
      CurveIndex=$Emitter.Age
     )
    )
    ScaleA=true
    ScaleRGB=true
   )
   &SolveForcesAndVelocity(Clamp Velocity=false, Limit Acceleration=false)
   &DynamicMaterialParameters(
    Index 0 Param 1=&FloatFromCurve(FloatCurve=DataInterface() {propertyValues {Curve {keys [
         {interpMode=RCIM_Cubic, leaveTangentWeight=0.3342, tangentMode=RCTM_User, time=0, value=0}
         {arriveTangent=-1.3198, arriveTangentWeight=0.3333, interpMode=RCIM_Cubic, leaveTangent=-1.3198, tangentMode=RCTM_User, time=1, value=-0.4908}
        ]}}})
    Index 0 Param 2=&RandomRangeFloat(Minimum=0, Maximum=100)
    Module.Param0WriteEnabled=true
    Module.Param1WriteEnabled=true
    Module.Param2WriteEnabled=false
    Module.Param3WriteEnabled=false
   )
  ]
  rendererClasses [&NiagaraSpriteRendererProperties]
  renderers [{rendererClass=&NiagaraSpriteRendererProperties}]
 }
 Unranked_Embers {
  emitterUpdateScript [
   &EmitterState()
   &SpawnBurst_Instantaneous(
    Spawn Count=&MakeCustomIntFromBool(Bool=$User.Unranked, True INT=75, False INT=0)
    Spawn Time=2.05
    Use Spawn Probability=false
    UseLoopCountLimit=false
   )
   &SpawnRate(
    SpawnRate=&MakeCustomFloatFromBool(
     Bool=$User.Unranked
     True Float=&Multiply_Float(
      A=&RandomRangeFloat(Minimum=6, Maximum=18, Evaluation Type=::NewEnumerator1)
      B=&FloatFromCurve(FloatCurve=DataInterface() {propertyValues {Curve {keys [<at 3.257 1>,<at 4 0>]}}}, CurveIndex=$Emitter.Age)
     )
     False Float=0
    )
    Use Spawn Probability=false
   )
  ]
  particleSpawnScript [
   &InitializeParticle(
    Color=&MakeLinearColorFromVectorAndFloat(
     "Vector (RGB)"=&VectorFromFloat(Value=&RandomRangeFloat(Minimum=2, Maximum=6))
     "Float (Alpha)"=&MakeCustomFloatFromBool(
      Bool=&SetBoolByFloatComparison(A=$User.Ranked_Curve, B=5, Comparison Type=::NewEnumerator4)
      True Float=10
      False Float=1
     )
    )
    Lifetime Min=0.3
    Lifetime Mode=::NewEnumerator1
    Position Offset=<-5 0 0>
    Sprite Size=&Vector2DFromFloat(Value=&FloatFromCurve(
      FloatCurve=DataInterface() {propertyValues {Curve {keys [<at 0 1.2>,<at 0.1257 0.9847>,<at 0.2996 0.5>,<at 1 0.3>]}}}
      CurveIndex=&RandomRangeFloat(Minimum=0, Maximum=1)
      Scale Curve=2
     ))
    Sprite Size Mode=::NewEnumerator4
    UsePositionOffset=true
    Write Material Random=true
   )
   &ShapeLocation(MaskSpawnGroup=false, Override Seed=false, Sphere Radius=8, UseOwnerScale=true)
   &AddVelocity(
    Module.Use Falloff Distance=false
    Origin Offset=<-5 0 0>
    Override Seed=false
    Velocity Mode=::NewEnumerator1
    Velocity Speed=&RandomRangeFloat(Minimum=25, Maximum=50)
   )
  ]
  particleUpdateScript [
   &ParticleState()
   &Drag(
    Drag=&FloatFromCurve(FloatCurve=DataInterface() {propertyValues {Curve {keys [<at 0 0.5>,<at 0.5 0.5>,<at 1 1>]}}}, Scale Curve=5)
    Use Linear Drag=true
    Use Rotational Drag=false
   )
   &SolveForcesAndVelocity(Clamp Velocity=false, Limit Acceleration=false)
   &ScaleSpriteSizeBySpeed(
    Max Scale Factor=<1 16>
    Min Scale Factor=<1 2>
    Sample Scale Factor By Curve=true
    Scale Factor Curve=DataInterface() {propertyValues {Curve {keys [
        {arriveTangent=2.5021, interpMode=RCIM_Cubic, leaveTangent=2.5021, leaveTangentWeight=0.6864, tangentMode=RCTM_User, time=0, value=0.01}
        {arriveTangentWeight=0.663, interpMode=RCIM_Cubic, tangentMode=RCTM_User, time=1, value=1}
       ]}}}
   )
   &ScaleColor(
    Scale Alpha=&FloatFromCurve(
     FloatCurve=DataInterface() {propertyValues {Curve {keys [<at 0 1>,<at 1 0>]}}}
     Scale Curve=&FloatFromCurve(
      FloatCurve=DataInterface() {propertyValues {Curve {keys [{arriveTangent=-2.4644, interpMode=RCIM_Cubic, leaveTangent=-2.4644, tangentMode=RCTM_User, time=3.257, value=1}, <at 4 0 RCIM_Cubic>]}}}
      CurveIndex=$Emitter.Age
     )
    )
    Scale RGB=&MakeVectorFromLinearColorRGB(Color=&MultiplyLinearColor(
      Linear Color=&self:SelectColorCurve(
       CurveSelect:NiagaraFloat=$User.Ranked_Curve
       ColorSelect:NiagaraFloat=&RandomRangeFloat(Minimum=0, Maximum=1)
       UnrankedColorCurve:NiagaraDataInterfaceColorCurve=DataInterface() {propertyValues {
         AlphaCurve {keys [<at 0 1>,<at 1 1>]}
         BlueCurve {keys [
           {interpMode=RCIM_Cubic, leaveTangentWeight=0.4171, tangentMode=RCTM_User, time=0, value=0.2831}
           {arriveTangent=1.9677, interpMode=RCIM_Cubic, leaveTangent=1.9677, tangentMode=RCTM_User, time=1, value=1}
          ]}
         GreenCurve {keys [
           {interpMode=RCIM_Cubic, leaveTangentWeight=0.4171, tangentMode=RCTM_User, time=0, value=0.2831}
           {arriveTangent=1.9677, interpMode=RCIM_Cubic, leaveTangent=1.9677, tangentMode=RCTM_User, time=1, value=1}
          ]}
         RedCurve {keys [
           {interpMode=RCIM_Cubic, leaveTangentWeight=0.4171, tangentMode=RCTM_User, time=0, value=0.2831}
           {arriveTangent=1.9677, interpMode=RCIM_Cubic, leaveTangent=1.9677, tangentMode=RCTM_User, time=1, value=1}
          ]}
        }}
       BronzeColorCurve:NiagaraDataInterfaceColorCurve=DataInterface() {propertyValues {
         AlphaCurve {keys [<at 0 1>,<at 1 1>]}
         BlueCurve {keys [<at 0 0.0185>,<at 1 0.1946>]}
         GreenCurve {keys [<at 0 0.0714>,<at 1 0.3557>]}
         RedCurve {keys [<at 0 0.2831>,<at 1 1>]}
        }}
       SilverColorCurve:NiagaraDataInterfaceColorCurve=DataInterface() {propertyValues {
         AlphaCurve {keys [<at 0 1>,<at 1 1>]}
         BlueCurve {keys [<at 0 0.0409>,<at 1 0.9734>]}
         GreenCurve {keys [<at 0 0.0273>,<at 1 0.9301>]}
         RedCurve {keys [<at 0 0.0194>,<at 1 0.8796>]}
        }}
       GoldColorCurve:NiagaraDataInterfaceColorCurve=DataInterface() {propertyValues {
         AlphaCurve {keys [<at 0 1>,<at 1 1>]}
         BlueCurve {keys [<at 0 0>,<at 1 0.2961>]}
         GreenCurve {keys [<at 0 0.1694>,<at 1 0.7654>]}
         RedCurve {keys [<at 0 0.2542>,<at 1 1>]}
        }}
       PlatColorCurve:NiagaraDataInterfaceColorCurve=DataInterface() {propertyValues {
         AlphaCurve {keys [<at 0 1>,<at 1 1>]}
         BlueCurve {keys [<at 0 0.0221>,<at 1 0.404>]}
         GreenCurve {keys [<at 0 0.05>,<at 1 0.9216>]}
         RedCurve {keys [<at 0 0.0021>,<at 1 0.0343>]}
        }}
       DiamondColorCurve:NiagaraDataInterfaceColorCurve=DataInterface() {propertyValues {
         AlphaCurve {keys [<at 0 1>,<at 1 1>]}
         BlueCurve {keys [<at 0 0.4>,<at 1 1>]}
         GreenCurve {keys [<at 0 0.2356>,<at 1 0.5833>]}
         RedCurve {keys [<at 0 0.0054>,<at 1 0>]}
        }}
       EliteColorCurve:NiagaraDataInterfaceColorCurve=DataInterface() {propertyValues {
         AlphaCurve {keys [<at 0 1>,<at 1 1>]}
         BlueCurve {keys [<at 0 0.0125 RCIM_Cubic>, {arriveTangent=0.3468, interpMode=RCIM_Cubic, leaveTangent=0.3468, tangentMode=RCTM_User, time=1, value=0.1}]}
         GreenCurve {keys [<at 0 0.0125 RCIM_Cubic>, {arriveTangent=0.3468, interpMode=RCIM_Cubic, leaveTangent=0.3468, tangentMode=RCTM_User, time=1, value=0.1}]}
         RedCurve {keys [<at 0 0.0125 RCIM_Cubic>, {arriveTangent=0.3468, interpMode=RCIM_Cubic, leaveTangent=0.3468, tangentMode=RCTM_User, time=1, value=0.1}]}
        }}
       ChampionColorCurve:NiagaraDataInterfaceColorCurve=DataInterface() {propertyValues {
         AlphaCurve {keys [<at 0 1>,<at 1 1>]}
         BlueCurve {keys [<at 0 0>,<at 0.5 0>,<at 1 0.0012>]}
         GreenCurve {keys [<at 0 0 RCIM_Cubic>, {arriveTangent=0.6654, interpMode=RCIM_Cubic, leaveTangent=0.6654, time=0.5, value=0.0908}, <at 1 0.6654 RCIM_Cubic>]}
         RedCurve {keys [
           <at 0 0.15 RCIM_Cubic>
           {arriveTangentWeight=0.2187, interpMode=RCIM_Cubic, leaveTangentWeight=0.2187, tangentMode=RCTM_User, time=0.5, value=1}
           <at 1 1 RCIM_Cubic>
          ]}
        }}
       UnrealColorCurve:NiagaraDataInterfaceColorCurve=DataInterface() {propertyValues {
         AlphaCurve {keys [<at 0 1>,<at 1 1>]}
         BlueCurve {keys [<at 0 0.2307 RCIM_Cubic>, {interpMode=RCIM_Cubic, leaveTangent=0.5013, time=1, value=0.2664}]}
         GreenCurve {keys [<at 0 0.0252 RCIM_Cubic>, {interpMode=RCIM_Cubic, leaveTangent=0.6649, time=1, value=0.0782}]}
         RedCurve {keys [
           <at 0 0.0222 RCIM_Cubic>
           {arriveTangentWeight=0.2187, interpMode=RCIM_Cubic, leaveTangent=0.4814, leaveTangentWeight=0.2187, time=1, value=0.3564}
          ]}
        }}
      )
      Scale Factor=&MakeVector4_FromFloat(Float=&RandomRangeFloat(Minimum=1, Maximum=2))
     ))
    ScaleA=true
    ScaleRGB=true
   )
  ]
  rendererClasses [&NiagaraSpriteRendererProperties]
  renderers [{rendererClass=&NiagaraSpriteRendererProperties}]
 }
 Unranked_Flicker_Flare {
  emitterUpdateScript [
   &EmitterState()
   &SpawnRate(
    SpawnRate=&Multiply_Float(
     A=&RandomRangeFloat(Minimum=8, Maximum=13)
     B=&FloatFromCurve(FloatCurve=DataInterface() {propertyValues {Curve {keys [<at 3.257 1>,<at 4 0>]}}}, CurveIndex=$Emitter.Age)
    )
    Use Spawn Probability=false
   )
   &SpawnRate(
    SpawnGroup=1
    SpawnRate=&MakeCustomFloatFromBool(
     Bool=$User.Unranked
     True Float=&Multiply_Float(
      A=&RandomRangeFloat(Minimum=8, Maximum=13)
      B=&FloatFromCurve(FloatCurve=DataInterface() {propertyValues {Curve {keys [<at 3.257 1>,<at 4 0>]}}}, CurveIndex=$Emitter.Age)
     )
     False Float=0
    )
    Use Spawn Probability=false
   ) {module=SpawnRate001}
  ]
  particleSpawnScript [
   &InitializeParticle(
    Color=&self:SelectColorCurve(
     CurveSelect:NiagaraFloat=$User.Ranked_Curve
     ColorSelect:NiagaraFloat=1
     UnrankedColorCurve:NiagaraDataInterfaceColorCurve=DataInterface() {propertyValues {
       AlphaCurve {keys [<at 0 1>,<at 1 1>]}
       BlueCurve {keys [
         {interpMode=RCIM_Cubic, leaveTangentWeight=0.4171, tangentMode=RCTM_User, time=0, value=0.2831}
         {arriveTangent=1.9677, interpMode=RCIM_Cubic, leaveTangent=1.9677, tangentMode=RCTM_User, time=1, value=1}
        ]}
       GreenCurve {keys [
         {interpMode=RCIM_Cubic, leaveTangentWeight=0.4171, tangentMode=RCTM_User, time=0, value=0.2831}
         {arriveTangent=1.9677, interpMode=RCIM_Cubic, leaveTangent=1.9677, tangentMode=RCTM_User, time=1, value=1}
        ]}
       RedCurve {keys [
         {interpMode=RCIM_Cubic, leaveTangentWeight=0.4171, tangentMode=RCTM_User, time=0, value=0.2831}
         {arriveTangent=1.9677, interpMode=RCIM_Cubic, leaveTangent=1.9677, tangentMode=RCTM_User, time=1, value=1}
        ]}
      }}
     BronzeColorCurve:NiagaraDataInterfaceColorCurve=DataInterface() {propertyValues {
       AlphaCurve {keys [<at 0 1>,<at 1 1>]}
       BlueCurve {keys [<at 0 0.0185>,<at 1 0.1946>]}
       GreenCurve {keys [<at 0 0.0714>,<at 1 0.3557>]}
       RedCurve {keys [<at 0 0.2831>,<at 1 1>]}
      }}
     SilverColorCurve:NiagaraDataInterfaceColorCurve=DataInterface() {propertyValues {
       AlphaCurve {keys [<at 0 1>,<at 1 1>]}
       BlueCurve {keys [<at 0 0.0409>,<at 1 0.9734>]}
       GreenCurve {keys [<at 0 0.0273>,<at 1 0.9301>]}
       RedCurve {keys [<at 0 0.0194>,<at 1 0.8796>]}
      }}
     GoldColorCurve:NiagaraDataInterfaceColorCurve=DataInterface() {propertyValues {
       AlphaCurve {keys [<at 0 1>,<at 1 1>]}
       BlueCurve {keys [<at 0 0>,<at 1 0.2961>]}
       GreenCurve {keys [<at 0 0.1694>,<at 1 0.7654>]}
       RedCurve {keys [<at 0 0.2542>,<at 1 1>]}
      }}
     PlatColorCurve:NiagaraDataInterfaceColorCurve=DataInterface() {propertyValues {
       AlphaCurve {keys [<at 0 1>,<at 1 1>]}
       BlueCurve {keys [<at 0 0.0221>,<at 1 0.404>]}
       GreenCurve {keys [<at 0 0.05>,<at 1 0.9216>]}
       RedCurve {keys [<at 0 0.0021>,<at 1 0.0343>]}
      }}
     DiamondColorCurve:NiagaraDataInterfaceColorCurve=DataInterface() {propertyValues {
       AlphaCurve {keys [<at 0 1>,<at 1 1>]}
       BlueCurve {keys [<at 0 0.4>,<at 1 1>]}
       GreenCurve {keys [<at 0 0.2356>,<at 1 0.5833>]}
       RedCurve {keys [<at 0 0.0054>,<at 1 0>]}
      }}
     EliteColorCurve:NiagaraDataInterfaceColorCurve=DataInterface() {propertyValues {
       AlphaCurve {keys [<at 0 5>]}
       BlueCurve {keys [<at 0 0.0005 RCIM_Cubic>, {arriveTangent=0.3468, interpMode=RCIM_Cubic, leaveTangent=0.3468, tangentMode=RCTM_User, time=1, value=0.1}]}
       GreenCurve {keys [<at 0 0.0005 RCIM_Cubic>, {arriveTangent=0.3468, interpMode=RCIM_Cubic, leaveTangent=0.3468, tangentMode=RCTM_User, time=1, value=0.1}]}
       RedCurve {keys [<at 0 0.0005 RCIM_Cubic>, {arriveTangent=0.3468, interpMode=RCIM_Cubic, leaveTangent=0.3468, tangentMode=RCTM_User, time=1, value=0.1}]}
      }}
     ChampionColorCurve:NiagaraDataInterfaceColorCurve=DataInterface() {propertyValues {
       AlphaCurve {keys [<at 0 1>,<at 1 1>]}
       BlueCurve {keys [<at 0 0>,<at 0.5 0>,<at 1 0.0012>]}
       GreenCurve {keys [<at 0 0 RCIM_Cubic>, {arriveTangent=0.6654, interpMode=RCIM_Cubic, leaveTangent=0.6654, time=0.5, value=0.0908}, <at 1 0.6654 RCIM_Cubic>]}
       RedCurve {keys [
         <at 0 0.15 RCIM_Cubic>
         {arriveTangentWeight=0.2187, interpMode=RCIM_Cubic, leaveTangentWeight=0.2187, tangentMode=RCTM_User, time=0.5, value=1}
         <at 1 1 RCIM_Cubic>
        ]}
      }}
     UnrealColorCurve:NiagaraDataInterfaceColorCurve=DataInterface() {propertyValues {
       AlphaCurve {keys [<at 0 1>,<at 1 1>]}
       BlueCurve {keys [<at 0 0.2307 RCIM_Cubic>, {interpMode=RCIM_Cubic, leaveTangent=0.5013, time=1, value=0.2664}]}
       GreenCurve {keys [<at 0 0.0252 RCIM_Cubic>, {interpMode=RCIM_Cubic, leaveTangent=0.6649, time=1, value=0.0782}]}
       RedCurve {keys [
         <at 0 0.0222 RCIM_Cubic>
         {arriveTangentWeight=0.2187, interpMode=RCIM_Cubic, leaveTangent=0.4814, leaveTangentWeight=0.2187, time=1, value=0.3564}
        ]}
      }}
    )
    Lifetime Max=1
    Lifetime Min=0.5
    Lifetime Mode=::NewEnumerator1
    Position Offset=<-5 0 0>
    Sprite Size Mode=::NewEnumerator3
    Uniform Sprite Size=&MaskFloatBySpawnGroup(Masked Float=75, Passthrough Float=35)
    UsePositionOffset=true
    Write Material Random=true
   )
   &CameraOffset(Camera Offset Amount=&MaskFloatBySpawnGroup(Masked Float=-5, Passthrough Float=15))
  ]
  particleUpdateScript [
   &ParticleState()
   &ScaleSpriteSize(
    Uniform Curve Index=$Emitter.Age
    Uniform Curve Sprite Scale=DataInterface() {propertyValues {Curve {keys [<at 0 0.5 RCIM_Constant>, {interpMode=RCIM_Cubic, tangentMode=RCTM_User, time=2.06, value=1.5}, <at 4 1 RCIM_Cubic>]}}}
   )
   &ScaleColor(
    Scale Alpha=&FloatFromCurve(
     FloatCurve=DataInterface() {propertyValues {Curve {keys [<at 0 1>,<at 1 0>]}}}
     Scale Curve=&FloatFromCurve(
      FloatCurve=DataInterface() {propertyValues {Curve {keys [{arriveTangent=-2.4644, interpMode=RCIM_Cubic, leaveTangent=-2.4644, tangentMode=RCTM_User, time=3.257, value=1}, <at 4 0 RCIM_Cubic>]}}}
      CurveIndex=$Emitter.Age
     )
    )
    ScaleA=true
    ScaleRGB=true
   )
   &DynamicMaterialParameters(
    Index 0 Param 1=0.75
    Index 0 Param 2=1
    Module.Param0WriteEnabled=true
    Module.Param1WriteEnabled=true
    Module.Param2WriteEnabled=true
    Module.Param3WriteEnabled=true
   )
  ]
  rendererClasses [&NiagaraSpriteRendererProperties]
  renderers [{rendererClass=&NiagaraSpriteRendererProperties}]
 }
 Unranked_LightShafts {
  emitterUpdateScript [
   &EmitterState()
   &SpawnRate(
    SpawnRate=&MakeCustomFloatFromBool(
     Bool=$User.Unranked
     True Float=&Multiply_Float(
      A=24
      B=&FloatFromCurve(FloatCurve=DataInterface() {propertyValues {Curve {keys [<at 3.257 1>,<at 4 0>]}}}, CurveIndex=$Emitter.Age)
     )
     False Float=0
    )
    Use Spawn Probability=false
   )
  ]
  particleSpawnScript [
   &InitializeParticle(
    Color=&self:SelectColorCurve(
     CurveSelect:NiagaraFloat=$User.Ranked_Curve
     ColorSelect:NiagaraFloat=&RandomRangeFloat(Minimum=0, Maximum=1)
     UnrankedColorCurve:NiagaraDataInterfaceColorCurve=DataInterface() {propertyValues {
       AlphaCurve {keys [<at 0 1>,<at 1 1>]}
       BlueCurve {keys [
         {interpMode=RCIM_Cubic, leaveTangentWeight=0.4171, tangentMode=RCTM_User, time=0, value=0.2831}
         {arriveTangent=1.9677, interpMode=RCIM_Cubic, leaveTangent=1.9677, tangentMode=RCTM_User, time=1, value=1}
        ]}
       GreenCurve {keys [
         {interpMode=RCIM_Cubic, leaveTangentWeight=0.4171, tangentMode=RCTM_User, time=0, value=0.2831}
         {arriveTangent=1.9677, interpMode=RCIM_Cubic, leaveTangent=1.9677, tangentMode=RCTM_User, time=1, value=1}
        ]}
       RedCurve {keys [
         {interpMode=RCIM_Cubic, leaveTangentWeight=0.4171, tangentMode=RCTM_User, time=0, value=0.2831}
         {arriveTangent=1.9677, interpMode=RCIM_Cubic, leaveTangent=1.9677, tangentMode=RCTM_User, time=1, value=1}
        ]}
      }}
     BronzeColorCurve:NiagaraDataInterfaceColorCurve=DataInterface() {propertyValues {
       AlphaCurve {keys [<at 0 1>,<at 1 1>]}
       BlueCurve {keys [<at 0 0.0185>,<at 1 0.1946>]}
       GreenCurve {keys [<at 0 0.0714>,<at 1 0.3557>]}
       RedCurve {keys [<at 0 0.2831>,<at 1 1>]}
      }}
     SilverColorCurve:NiagaraDataInterfaceColorCurve=DataInterface() {propertyValues {
       AlphaCurve {keys [<at 0 1>,<at 1 1>]}
       BlueCurve {keys [<at 0 0.0409>,<at 1 0.9734>]}
       GreenCurve {keys [<at 0 0.0273>,<at 1 0.9301>]}
       RedCurve {keys [<at 0 0.0194>,<at 1 0.8796>]}
      }}
     GoldColorCurve:NiagaraDataInterfaceColorCurve=DataInterface() {propertyValues {
       AlphaCurve {keys [<at 0 1>,<at 1 1>]}
       BlueCurve {keys [<at 0 0>,<at 1 0.2961>]}
       GreenCurve {keys [<at 0 0.1694>,<at 1 0.7654>]}
       RedCurve {keys [<at 0 0.2542>,<at 1 1>]}
      }}
     PlatColorCurve:NiagaraDataInterfaceColorCurve=DataInterface() {propertyValues {
       AlphaCurve {keys [<at 0 1>,<at 1 1>]}
       BlueCurve {keys [<at 0 0.0221>,<at 1 0.404>]}
       GreenCurve {keys [<at 0 0.05>,<at 1 0.9216>]}
       RedCurve {keys [<at 0 0.0021>,<at 1 0.0343>]}
      }}
     DiamondColorCurve:NiagaraDataInterfaceColorCurve=DataInterface() {propertyValues {
       AlphaCurve {keys [<at 0 1>,<at 1 1>]}
       BlueCurve {keys [<at 0 0.4>,<at 1 1>]}
       GreenCurve {keys [<at 0 0.2356>,<at 1 0.5833>]}
       RedCurve {keys [<at 0 0.0054>,<at 1 0>]}
      }}
     EliteColorCurve:NiagaraDataInterfaceColorCurve=DataInterface() {propertyValues {
       AlphaCurve {keys [<at 0 25>]}
       BlueCurve {keys [<at 0 0.002 RCIM_Cubic>]}
       GreenCurve {keys [<at 0 0.002 RCIM_Cubic>]}
       RedCurve {keys [<at 0 0.002 RCIM_Cubic>]}
      }}
     ChampionColorCurve:NiagaraDataInterfaceColorCurve=DataInterface() {propertyValues {
       AlphaCurve {keys [<at 0 1>,<at 1 1>]}
       BlueCurve {keys [<at 0 0>,<at 0.5 0>,<at 1 0.0012>]}
       GreenCurve {keys [<at 0 0 RCIM_Cubic>, {arriveTangent=0.6654, interpMode=RCIM_Cubic, leaveTangent=0.6654, time=0.5, value=0.0908}, <at 1 0.6654 RCIM_Cubic>]}
       RedCurve {keys [
         <at 0 0.15 RCIM_Cubic>
         {arriveTangentWeight=0.2187, interpMode=RCIM_Cubic, leaveTangentWeight=0.2187, tangentMode=RCTM_User, time=0.5, value=1}
         <at 1 1 RCIM_Cubic>
        ]}
      }}
     UnrealColorCurve:NiagaraDataInterfaceColorCurve=DataInterface() {propertyValues {
       AlphaCurve {keys [<at 0 1>,<at 1 1>]}
       BlueCurve {keys [<at 0 0.2307 RCIM_Cubic>, {interpMode=RCIM_Cubic, leaveTangent=0.5013, time=1, value=0.2664}]}
       GreenCurve {keys [<at 0 0.0252 RCIM_Cubic>, {interpMode=RCIM_Cubic, leaveTangent=0.6649, time=1, value=0.0782}]}
       RedCurve {keys [
         <at 0 0.0222 RCIM_Cubic>
         {arriveTangentWeight=0.2187, interpMode=RCIM_Cubic, leaveTangent=0.4814, leaveTangentWeight=0.2187, time=1, value=0.3564}
        ]}
      }}
    )
    Lifetime Max=1
    Lifetime Min=0.5
    Lifetime Mode=::NewEnumerator1
    Position Offset=<-5 0 0>
    Sprite Size Max=<55 65>
    Sprite Size Min=<30 45>
    Sprite Size Mode=::NewEnumerator2
    UsePositionOffset=true
    Write Material Random=true
   )
   &CameraOffset(Camera Offset Amount=-10)
   &AddVelocity(
    Module.Use Falloff Distance=false
    Origin Offset=<-5 0 0>
    Origin Offset Coordinate Space=::Simulation
    Override Seed=false
    Velocity Mode=::NewEnumerator1
    Velocity Speed=1
   )
   &SubUVAnimation(
    Evaluation Type=::NewEnumerator0
    "Sprite Renderer "=DataInterface() {propertyValues {SpriteRenderer=&self:Unranked_LightShafts_3.NiagaraSpriteRendererProperties_0}}
    SubUV Animation Mode:ENiagara_SubUVLookupModeV2=::NewEnumerator2
    UseEndFrame=false
    UseStartFrame=false
   )
  ]
  particleUpdateScript [
   &ParticleState()
   &ScaleSpriteSize(
    Non-Uniform Curve Scale=&MakeVector2D(
     X=1
     Y=&FloatFromCurve(
      FloatCurve=DataInterface() {propertyValues {Curve {keys [<at 1.75 0 RCIM_Cubic>, {interpMode=RCIM_Cubic, leaveTangent=1145.6222, time=2.25, value=1}]}}}
      CurveIndex=$Emitter.Age
     )
    )
    Non-Uniform Curve Sprite Scale=DataInterface() {propertyValues {
      XCurve {keys [{tangentMode=RCTM_User, time=0, value=1}, {tangentMode=RCTM_User, time=1, value=0.75}]}
      YCurve {keys [{tangentMode=RCTM_User, time=0, value=0.25}, {tangentMode=RCTM_User, time=1, value=1.5}]}
     }}
    Scale Sprite Size Mode=::NewEnumerator3
   )
   &SolveForcesAndVelocity(Clamp Velocity=false, Limit Acceleration=false)
   &ScaleColor(
    Scale Alpha=&FloatFromCurve(
     FloatCurve=DataInterface() {propertyValues {Curve {keys [
         {arriveTangent=9.0725, interpMode=RCIM_Cubic, leaveTangent=9.0725, tangentMode=RCTM_User, time=0, value=0}
         {arriveTangent=-2.5249, interpMode=RCIM_Cubic, leaveTangent=-2.5249, tangentMode=RCTM_User, time=0.1408, value=0.9847}
         {arriveTangent=-0.0405, interpMode=RCIM_Cubic, leaveTangent=-0.0405, tangentMode=RCTM_User, time=1, value=0}
        ]}}}
     Scale Curve=&FloatFromCurve(
      FloatCurve=DataInterface() {propertyValues {Curve {keys [<at 1.75 0 RCIM_Cubic>, {interpMode=RCIM_Cubic, leaveTangent=1145.6222, time=2.25, value=1}]}}}
      CurveIndex=$Emitter.Age
      Scale Curve=&FloatFromCurve(
       FloatCurve=DataInterface() {propertyValues {Curve {keys [{arriveTangent=-2.4644, interpMode=RCIM_Cubic, leaveTangent=-2.4644, tangentMode=RCTM_User, time=3.257, value=1}, <at 4 0 RCIM_Cubic>]}}}
       CurveIndex=$Emitter.Age
      )
     )
    )
    Scale RGB=&MakeVectorFromLinearColorRGB(Color=&ColorFromCurve(ColorCurve=DataInterface() {propertyValues {
        AlphaCurve {keys []}
        BlueCurve {keys [<at 0 1>,<at 0.2 2>,<at 0.65 1>,<at 1 0>]}
        GreenCurve {keys [<at 0 1>,<at 0.2 2>,<at 0.65 1>,<at 1 0>]}
        RedCurve {keys [<at 0 1>,<at 0.2 2>,<at 0.65 1>,<at 1 0>]}
       }}))
    ScaleA=true
    ScaleRGB=true
   )
   &ScaleColor(
    Scale Alpha=&Multiply_Float(
     A=&RandomRangeFloat(Minimum=0.25, Maximum=0.5)
     B=&FloatFromCurve(
      FloatCurve=DataInterface() {propertyValues {Curve {keys [{arriveTangent=-3.7897, interpMode=RCIM_Cubic, leaveTangent=-3.7897, tangentMode=RCTM_User, time=4.572, value=1}, <at 5 0>]}}}
      CurveIndex=$Emitter.Age
     )
    )
    Scale RGB=&VectorFromFloat(Value=10)
    ScaleA=true
    ScaleRGB=true
   ) {module=ScaleColor001}
  ]
  rendererClasses [&NiagaraSpriteRendererProperties]
  renderers [{rendererClass=&NiagaraSpriteRendererProperties}]
 }
}
summary {
 systemName=NS_RankedProgression_FlareUp_v3
 emitters [
  {emitterName=Flamin_Ribbon, rendererClasses [&NiagaraRibbonRendererProperties]}
  {emitterName=Flame_Meshes, rendererClasses [&NiagaraMeshRendererProperties]}
  {emitterName=Core_Fire_Sprites, rendererClasses [&NiagaraSpriteRendererProperties]}
  {emitterName=Flames_BKGD, rendererClasses [&NiagaraSpriteRendererProperties]}
  {emitterName=Flame_Decals, rendererClasses [&NiagaraSpriteRendererProperties]}
  {emitterName=Embers, rendererClasses [&NiagaraSpriteRendererProperties]}
  {
   emitterName=Light
   rendererClasses [
    &NiagaraSpriteRendererProperties
    &NiagaraComponentRendererProperties
    &NiagaraLightRendererProperties
   ]
  }
  {emitterName=BottomFireSprites, rendererClasses [&NiagaraSpriteRendererProperties]}
  {emitterName=Unranked_BKGD, rendererClasses [&NiagaraSpriteRendererProperties]}
  {emitterName=Unranked_LightShafts, rendererClasses [&NiagaraSpriteRendererProperties]}
  {emitterName=Unranked_Flicker_Flare, rendererClasses [&NiagaraSpriteRendererProperties]}
  {emitterName=Unranked_Embers, rendererClasses [&NiagaraSpriteRendererProperties]}
  {emitterName=Fire_Strokes, rendererClasses [&NiagaraSpriteRendererProperties]}
  {emitterName=Corona_Burst, rendererClasses [&NiagaraSpriteRendererProperties]}
 ]
 userVariables [
  User.Ranked_Curve {type=NiagaraFloat, defaultValue=0, description=""}
  User.Unranked {type=NiagaraBool, defaultValue=true}
 ]
}
